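# MF_UVMask01Space is a material function for the F_MED_WingedEye cosmetic body that builds a binary mask of the 0-1 UV square.
# It splits the UV input into U and V and tests each with two If nodes against bounds inset slightly from 0 and 1.
# The four results are multiplied, outputting 1 only where both coordinates fall inside those bounds and 0 elsewhere.
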MaterialFunction /BRCosmetics/Characters/Player/Female/Medium/Bodies/F_MED_WingedEye/Materials/MF_UVMask01Space
graph {nodes [
  n1 = FunctionInput() {inputName=UV, inputType=FunctionInput_Vector2}
  n2 = ComponentMask(None=@n1) {r=true}
  n3 = Constant() {r=1}
  n4 = Constant()
  n5 = If(A=@n2, "A > B"=@n3, "A == B"=@n3, "A < B"=@n4) {constB=0.01}
  n6 = If(A=@n2, "A > B"=@n4, "A == B"=@n3, "A < B"=@n3) {constB=0.99}
  n7 = ComponentMask(None=@n1) {g=true}
  n8 = Constant() {r=1}
  n9 = Constant()
  n10 = If(A=@n7, "A > B"=@n8, "A == B"=@n8, "A < B"=@n9) {constB=0.01}
  n11 = If(A=@n7, "A > B"=@n9, "A == B"=@n8, "A < B"=@n8) {constB=0.99}
  FunctionOutput(None=Multiply(A=Multiply(A=@n5, B=@n6), B=Multiply(A=@n10, B=@n11)))
 ]}
