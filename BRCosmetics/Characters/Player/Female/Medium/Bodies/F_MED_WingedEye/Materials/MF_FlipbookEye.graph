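# Material function for the Winged Eye cosmetic, outputting color and alpha from a 4x4 eye flipbook.
# Facial pose parameters crossing thresholds select the frame, with static switches choosing max-combine, priority-chain or dominant-pose logic.
# The dominant-pose branch (Logic03) replaces pose-driven blinking with a periodic, sine-modulated timer blink.
# A custom node positions and scales UVs using only CenterX and ScaleX for both axes, and a UV-bounds mask trims alpha.
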MaterialFunction /BRCosmetics/Characters/Player/Female/Medium/Bodies/F_MED_WingedEye/Materials/MF_FlipbookEye
graph {nodes [
  n1 = NamedRerouteUsage()
  n2 = ScalarParameter() {group=EyeFlipbook, parameterName=DefaultEye}
  n3 = Max(
   A=If(
    A=ScalarParameter() {group=Banana_Poses, parameterName=R_brow_up_pose}
    B=NamedRerouteUsage()
    "A > B"=ScalarParameter() {defaultValue=3, group=Banana, parameterName="BananaBrowUpValue(R)"}
    "A == B"=@n2
    "A < B"=@n2
   )
   B=If(
    A=ScalarParameter() {group=Banana_Poses, parameterName=C_glabella_down_pose}
    B=NamedRerouteUsage()
    "A > B"=ScalarParameter() {defaultValue=5, group=Banana, parameterName="GlabellaDownValue(G)"}
    "A == B"=@n2
    "A < B"=@n2
   )
  )
  n4 = If(
   A=ScalarParameter() {group=Banana_Poses, parameterName=C_glabella_up_pose}
   B=NamedRerouteUsage()
   "A > B"=Constant() {r=4}
   "A == B"=@n2
   "A < B"=@n2
  )
  n5 = ScalarParameter() {group=Banana_Poses, parameterName=R_blink_pose}
  n6 = NamedRerouteUsage()
  n7 = Constant() {r=2}
  n8 = NamedRerouteUsage()
  n9 = If(
   A=ScalarParameter() {group=Banana_Poses, parameterName=R_squint_inner_pose}
   B=NamedRerouteUsage()
   "A > B"=Constant() {r=1}
   "A == B"=@n8
   "A < B"=@n8
  )
  n10 = StaticSwitchParameter(True=NamedRerouteUsage(), False=Max(A=Max(A=@n3, B=@n4), B=If(A=@n5, B=@n6, "A > B"=@n7, "A == B"=@n9, "A < B"=@n9))) {defaultValue=true, parameterName=UseLogic02}
  n11 = Constant() {r=4}
  n12 = VectorParameter() {group=EyeFlipbook, parameterName=Center, channelNames {a=ScaleY, b=ScaleX, g=CenterY, r=CenterX}, defaultValue=<rgba 0.5 0.5 1 1>}
  MaterialExpressionAppendVector_7 = AppendVector(A=@n12.CenterX, B=@n12.CenterX)
  MaterialExpressionAppendVector_8 = AppendVector(A=@n12.ScaleX, B=@n12.ScaleX)
  MaterialExpressionTextureCoordinate_0 = TextureCoordinate()
  n13 = Custom!(UVPoint=@MaterialExpressionAppendVector_7, UVScale=@MaterialExpressionAppendVector_8, OriginalUV=@MaterialExpressionTextureCoordinate_0) {
   code="float2 DistUV =OriginalUV-UVPoint;\r\nDistUV = DistUV/UVScale + float2(0.5,0.5);\r\n\r\nreturn DistUV;"
   description=CopyToPointUV
   outputType=CMOT_Float2
   inputs [
    {
     inputName=UVPoint
     input {expression=&self:MaterialExpressionAppendVector_7}
    }
    {
     inputName=UVScale
     input {expression=&self:MaterialExpressionAppendVector_8}
    }
    {
     inputName=OriginalUV
     input {expression=&self:MaterialExpressionTextureCoordinate_0}
    }
   ]
  }
  n14 = AppendVector(A=ComponentMask(None=@n13) {r=true}, B=Multiply(A=ComponentMask(None=@n13) {g=true}) {constB=-1})
  n15 = MaterialFunctionCall(
   "Animation  Phase (0-1)"=Divide(A=StaticSwitchParameter(True=@n1, False=@n10) {defaultValue=true, parameterName=UseLogic03}) {constB=16}
   Number of Rows=@n11
   Number of Columns=@n11
   Texture=TextureObject() {texture=/BRCosmetics/Characters/Player/Female/Medium/Bodies/F_MED_WingedEye/Textures/T_Winged_Eye}
   UVs=@n14
  ) {materialFunction=&FlipBook}
  FunctionOutput(None=@n15.Result) {outputName=Color}
  FunctionOutput(None=Multiply(
    A=@n15.Alpha
    B=MaterialFunctionCall!(UV=@n14) {materialFunction=&/MF_UVMask01Space}.Result
   )) {outputName=Alpha, sortPriority=1}
  NamedRerouteDeclaration(None=ScalarParameter() {defaultValue=0.1, group=Banana, parameterName=Threshold_R_brow_up_pose}) {name=Threshold_Brow_Up, nodeColor=<rgba 0 0.6016 1 1>}
  NamedRerouteDeclaration(None=ScalarParameter() {defaultValue=0.1, group=Banana, parameterName=Threshold_R_blink_pose}) {name=Threshold_Blink, nodeColor=<rgba 0 0.6016 1 1>}
  NamedRerouteDeclaration(None=ScalarParameter() {defaultValue=0.1, group=Banana, parameterName=Threshold_Glabella_up}) {name=Threshold_Glabella_Up, nodeColor=<rgba 0 0.6016 1 1>}
  NamedRerouteDeclaration(None=ScalarParameter() {defaultValue=0.1, group=Banana, parameterName=Threshold_Glabella_down}) {name=Threshold_Glabella_Down, nodeColor=<rgba 0 0.6016 1 1>}
  NamedRerouteDeclaration(None=ScalarParameter() {defaultValue=0.375, group=Banana, parameterName=Threshold_squint_inner_pose}) {name=Threshold_Squint_Inner, nodeColor=<rgba 0 0.6016 1 1>}
  NamedRerouteDeclaration(None=@n2) {name=DefaultEye, nodeColor=<rgba 1 0.1875 0 1>}
  n16 = If(
   A=ScalarParameter() {group=Banana_Poses, parameterName=R_brow_up_pose}
   B=NamedRerouteUsage()
   "A > B"=ScalarParameter() {defaultValue=3, group=Banana, parameterName="BananaBrowUpValue(R)"}
   "A == B"=@n9
   "A < B"=@n9
  )
  n17 = If(
   A=ScalarParameter() {group=Banana_Poses, parameterName=C_glabella_down_pose}
   B=NamedRerouteUsage()
   "A > B"=ScalarParameter() {defaultValue=5, group=Banana, parameterName="GlabellaDownValue(G)"}
   "A == B"=@n16
   "A < B"=@n16
  )
  n18 = If(
   A=ScalarParameter() {group=Banana_Poses, parameterName=C_glabella_up_pose}
   B=NamedRerouteUsage()
   "A > B"=Constant() {r=4}
   "A == B"=@n17
   "A < B"=@n17
  )
  NamedRerouteDeclaration(None=If(
    A=ScalarParameter() {group=Banana_Poses, parameterName=R_blink_pose}
    B=NamedRerouteUsage()
    "A > B"=Constant() {r=2}
    "A == B"=@n18
    "A < B"=@n18
   )) {name=Logic02, nodeColor=<rgba 1 0 0.6328 1>}
  n19 = ScalarParameter() {group=Banana_Poses, parameterName=R_squint_inner_pose}
  n20 = ScalarParameter() {group=Banana_Poses, parameterName=R_brow_up_pose}
  NamedRerouteDeclaration(None=Max(A=@n19, B=@n20)) {name=Bigger01, nodeColor=<rgba 0 0.8125 1 1>}
  n21 = ScalarParameter() {group=Banana_Poses, parameterName=C_glabella_down_pose}
  n22 = ScalarParameter() {group=Banana_Poses, parameterName=C_glabella_up_pose}
  NamedRerouteDeclaration(None=Max(A=@n21, B=@n22)) {name=Bigger02, nodeColor=<rgba 0 0.8125 1 1>}
  n23 = NamedRerouteUsage()
  n24 = NamedRerouteUsage()
  n25 = NamedRerouteUsage()
  n26 = NamedRerouteUsage()
  n27 = If(
   A=@n20
   B=NamedRerouteUsage()
   "A > B"=ScalarParameter() {defaultValue=3, group=Banana, parameterName="BananaBrowUpValue(R)"}
   "A == B"=@n26
   "A < B"=@n26
  )
  n28 = If(
   A=@n19
   B=@n20
   "A > B"=If(A=@n19, B=NamedRerouteUsage(), "A > B"=Constant() {r=1}, "A == B"=@n25, "A < B"=@n25)
   "A == B"=@n27
   "A < B"=@n27
  )
  n29 = NamedRerouteUsage()
  n30 = NamedRerouteUsage()
  n31 = If(A=@n22, B=NamedRerouteUsage(), "A > B"=Constant() {r=4}, "A == B"=@n30, "A < B"=@n30)
  n32 = If(
   A=@n21
   B=@n22
   "A > B"=If(
    A=@n21
    B=NamedRerouteUsage()
    "A > B"=ScalarParameter() {defaultValue=5, group=Banana, parameterName="GlabellaDownValue(G)"}
    "A == B"=@n29
    "A < B"=@n29
   )
   "A == B"=@n31
   "A < B"=@n31
  )
  n33 = ScalarParameter() {defaultValue=0.9, parameterName=EyeThreshold}
  n34 = Time()
  n35 = Sine(None=Multiply(A=@n34, B=ScalarParameter() {defaultValue=0.1, group=EyeFlipbook, parameterName=SineSpeed}))
  n36 = ScalarParameter() {defaultValue=0.4, group=EyeFlipbook, parameterName=TimeSpeed}
  n37 = Frac(None=Add(A=ConstantBiasScale(None=@n35), B=Multiply(A=@n34, B=@n36)))
  NamedRerouteDeclaration(None=LinearInterpolate(A=If(A=@n23, B=@n24, "A > B"=@n28, "A == B"=@n32, "A < B"=@n32), B=Constant() {r=2}, Alpha=Step(Y=@n33, X=@n37) {constY=0.8})) {name=Logic03, nodeColor=<rgba 0.3359 1 0 1>}
  Step(Y=Subtract(A=@n33) {constB=0.3}, X=@n37) {constY=0.7}
 ]}
