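# Character material for the female WingedEye body that composites a flipbook eye into the 2023 anime shader, using the Cloth shading model.
# It chains the shared Pre, Calculations and Post functions, with MF_FlipbookEye supplying color and an alpha mask.
# That mask flattens the normal and blends in overrides for shading threshold, specular, roughness and lit and shaded colors.
# PBR_Shading blends the eye's shade color between anime and PBR tints.
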Material /BRCosmetics/Characters/Player/Female/Medium/Bodies/F_MED_WingedEye/Materials/M_WingedEye_Flipbook_2023
graph {
 properties {
  bAutomaticallySetUsageInEditor=false
  bUseMaterialAttributes=true
  bUsedWithMeshDeformer=true
  bUsedWithMorphTargets=true
  bUsedWithSkeletalMesh=true
  bUsedWithStaticLighting=true
  pixelDepthOffsetMode=PDOM_Legacy
  shadingModel=MSM_Cloth
  usedShadingModels=Cloth
 }
 nodes [
  Constant() {r=1}
  n1 = MaterialFunctionCall!(CustomUV) {materialFunction=/Game/Global/Materials/Character_Parent_2023/Anime_Parent_2023/MaterialFunctions/MFC_AnimeShader_2023_Pre}
  n2 = MaterialFunctionCall(Float3=@n1.STT) {materialFunction=&BreakOutFloat3Components}
  n3 = ScalarParameter() {defaultValue=0.5, parameterName=ThresholdOverride}
  n4 = MaterialFunctionCall() {materialFunction=&/MF_FlipbookEye}
  n5 = MaterialFunctionCall!(
   CustomLightVector
   OverrideNormal_WS
   STT=MaterialFunctionCall(X=@n2.R, Y=LinearInterpolate(A=@n2.G, B=@n3, Alpha=@n4.Alpha), Z=@n2.R) {materialFunction=&MakeFloat3}.Result
   Normal_Map=LinearInterpolate(A=@n1.Normal, B=Constant3Vector() {constant=<rgba 0 0 1 0>}, Alpha=@n4.Alpha)
   DFL=@n1.DFL
   FX=@n1.DFL
  ) {materialFunction=/Game/Global/Materials/Character_Parent_2023/Anime_Parent_2023/MaterialFunctions/MFC_AnimeShader_2023_Calculations}
  n6 = VectorParameter() {group=ColorOverride, parameterName=ColorLit, defaultValue=<rgba 0.8781 0.8781 0.8781 1>}
  n7 = Multiply(A=@n4.Color, B=VectorParameter() {group=ColorOverride, parameterName=ColorShadeTint, defaultValue=<rgba 0 0 0 1>}.RGB)
  n8 = Multiply(A=@n4.Color, B=VectorParameter() {group=ColorOverride, parameterName=ColorShadePBR, defaultValue=<rgba 1 1 1 1>}.RGB)
  n9 = Saturate(None=ScalarParameter() {group=03_Anime_PBR_Override, parameterName=PBR_Shading, sliderMax=1, sortPriority=5})
  n10 = MaterialFunctionCall(Float3=@n1.SRM) {materialFunction=&BreakOutFloat3Components}
  n11 = ScalarParameter() {defaultValue=0.5, group=ColorOverride, parameterName=Specular Override}
  n12 = ScalarParameter() {defaultValue=1, group=ColorOverride, parameterName=Roughness Override}
  MaterialFunctionCall!(
   Shading_Gradient=@n5.Shading_Gradient
   Threshold_Gradient=@n5.Shading_Gradient
   Specular_Gradient=@n5.Shading_Gradient
   Sheen_Mask=@n5.Shading_Gradient
   Thickness=@n5.Shading_Gradient
   DFL_Mask=@n5.Shading_Gradient
   FX=@n5.Shading_Gradient
   Metallic_Gradient=@n5.Shading_Gradient
   RimLight_Gradient=@n5.Shading_Gradient
   NormalBuffer=@n5.Shading_Gradient
   --------
   CINE_LightSources=@n5.Shading_Gradient
   CINE_LightColor=@n5.Shading_Gradient
   ---------
   Color_Lit=LinearInterpolate(A=@n1.Color_Lit, B=Multiply(A=@n4.Color, B=@n6.RGB), Alpha=@n4.Alpha)
   Color_Shaded=LinearInterpolate(A=@n1.Color_Shaded, B=LinearInterpolate(A=@n7, B=@n8, Alpha=@n9), Alpha=@n4.Alpha)
   SRM=MaterialFunctionCall(X=LinearInterpolate(A=@n10.R, B=@n11, Alpha=@n4.Alpha), Y=LinearInterpolate(A=@n10.G, B=@n12, Alpha=@n4.Alpha), Z=@n10.B) {materialFunction=&MakeFloat3}.Result
   Emissive=@n1.Emissive
   DFL_Color=@n1.Emissive
   Specular_Color=@n1.Emissive
   RimLight_Color=@n1.Emissive
  ) {materialFunction=/Game/Global/Materials/Character_Parent_2023/Anime_Parent_2023/MaterialFunctions/MFC_AnimeShader_2023_Post}
 ]
}
