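# Control Rig for the WingedEye medium female player body, adding procedural corrective deformation to body and clothing bones.
# A shared DeformRig function runs first, then a sequence derives sleeve correctives and inverse-bend curves from elbow yaw, varying by isLoFi and LOD.
# Spherical pose readers nudge sleeve and pant bones, a thigh attribute drives the front chain, and DBNO rotates the back hoodie.
# Foot and side-hoodie correctives are present but unconnected to execution.
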ControlRigBlueprint /BRCosmetics/Characters/Player/Female/Medium/Bodies/F_MED_WingedEye/Meshes/F_MED_WingedEye_CtrlRig
graphs [
 ContainedGraph {nodes [
   n1 = RigVMFunctionEntryNode()
   n2 = &RigUnit_GetTransform(Space=LocalSpace, Item=@n1.Item)
   n3 = &RigUnit_GetTransform(Space=LocalSpace, bInitial=true, Item=@n1.Item)
   > RigVMFunctionReturnNode(OutTransform=/Script/RigVM.RigVMFunction_MathTransformToEulerTransform(Value=&RigVMFunction_MathTransformMakeRelative(Global=@n2.Transform, Parent=@n3.Transform).Local) {is=struct RigVMUnitNode, template="Cast::Execute(in Value,out Result)"}.Result)
  ]}
 RigVMFunctionLibrary {nodes [DeformRig_GetDeltaRefTransform_FNC = RigVMCollapseNode(ExecuteContext=(), Item=(Type=Bone))]}
 RigVMFunction_Sequence_2_ContainedGraph {nodes [
   n1 = RigVMFunctionEntryNode()
   RigVMFunction_Sequence = &RigVMFunction_Sequence(ExecuteContext=@n1.ExecuteContext)
   n2 = &RigVMFunction_Sequence(ExecuteContext=@RigVMFunction_Sequence.B)
   n3 = &RigVMFunction_Sequence(ExecuteContext=@n2.B)
   n4 = &RigVMFunction_Sequence(ExecuteContext=@n3.B)
   RigVMFunctionReturnNode(A=@RigVMFunction_Sequence.A, B=@n2.A, C=@n3.A, D=@n4.A, E=@n4.B)
  ]}
 RigVMFunction_Sequence_3_ContainedGraph {nodes [
   n1 = RigVMFunctionEntryNode()
   n2 = &RigVMFunction_Sequence(ExecuteContext=@n1.ExecuteContext)
   n3 = &RigVMFunction_Sequence(ExecuteContext=@n2.B)
   RigVMFunctionReturnNode(A=@n2.A, B=@n3.A, C=@n3.B)
  ]}
 RigVMModel {nodes [
   n1 = &RigUnit_GetTransform(Item=(Type=Bone,Name="foot_r"), Space=LocalSpace, bInitial=False)
   &RigUnit_ModifyTransforms(
    ExecutePin=()
    ItemToModify=((Item=(Type=Bone,Name="deform_foot_bk_r"),Transform=(Rotation=(X=0.0644,Y=0.0453,Z=0.3246,W=-0.9426),Translation=(X=6.5683,Y=-3.8648,Z=-0.2327),Scale3D=(X=1,Y=1,Z=1))))
    WeightMinimum=0
    Weight=&RigVMFunction_MathDoubleRemap(
     SourceMaximum=65
     SourceMinimum=0
     TargetMaximum=1
     TargetMinimum=0
     Value=&RigVMFunction_MathQuaternionToEuler(RotationOrder=ZYX, Value=@n1.Transform.Rotation).Result.Z
    ).Result
   )
   n2 = &RigUnit_GetTransform(Item=(Type=Bone,Name="foot_l"), Space=LocalSpace, bInitial=False)
   > &RigUnit_ModifyTransforms(
    ItemToModify=((Item=(Type=Bone,Name="deform_foot_bk_l"),Transform=(Rotation=(X=0.0926,Y=-0.0867,Z=-0.1063,W=0.9862),Translation=(X=5.9568,Y=-2.6848,Z=-0.4558),Scale3D=(X=1,Y=1,Z=1))))
    WeightMinimum=0
    Weight=&RigVMFunction_MathDoubleRemap(
     SourceMaximum=65
     SourceMinimum=0
     TargetMaximum=1
     TargetMinimum=0
     Value=&RigVMFunction_MathQuaternionToEuler(RotationOrder=ZYX, Value=@n2.Transform.Rotation).Result.Z
    ).Result
   )
   RigUnit_BeginExecution = &RigUnit_BeginExecution()
   FN_DeformRig_FNC = RigVMFunctionReferenceNode(
    DeformRig_ Curve="(deformRig_l_upperarm_up_alpha_72_#1=False,deformRig_r_upperarm_up_alpha_73_#2=false,deformRig_l_upperarm_fr_alpha_74_#3=False,deformRig_r_upperarm_fr_alpha_75_CB51B8 ...(truncated)"
    DeformRig_Bool="(upperarm_1_#4=True,elbow_3_#5=True,wrist_7_#6=True,groin_8_#7=True,glute_17_#8=True,knee_9_40C408324CA261 ...(truncated)"
    DeformRig_Debug=(upperarm_debug_18_#4=False,wrist_debug_20_#6=False,lat_debug_26_#9=False,groinGlute_debug_27_#7=False)
    DeformRig_Value="(upperarm_up_18_#10=(Rotation=(X=0,Y=0,Z=0,W=1),Translation=(X=0,Y=0,Z=9),Scale3D=(X=0.7,Y=1,Z=1)),upperarm_fwd_fr_19_#11=(Rotation=(X=0,Y=0,Z=0,W=1),Translation=(X=0,Y=14,Z=0),Scale3D=(X=0.5,Y=1,Z= ...(truncated)"
    ExecuteContext=@RigUnit_BeginExecution.ExecutePin
   )
   > RigVMFunction_Sequence = RigVMAggregateNode()
   RigVMFunction_Sequence_1 = RigVMAggregateNode(ExecuteContext=@RigVMFunction_Sequence.A)
   DeformRig_GetDeltaRefTransform_FNC = RigVMFunctionReferenceNode(Item=(Type=Bone,Name="lowerarm_l"), ExecuteContext=@RigVMFunction_Sequence_1.A)
   n3 = &RigVMFunction_MathFloatRemap(
    SourceMaximum=110
    SourceMinimum=70
    TargetMaximum=5
    TargetMinimum=0
    bClamp=true
    Value=@DeformRig_GetDeltaRefTransform_FNC.OutTransform.Rotation.Yaw
   )
   Multiply = &RigVMFunction_MathDoubleMul(
    B=&RigVMFunction_MathFloatRemap(
     SourceMaximum=110
     SourceMinimum=70
     TargetMaximum=-0.08
     TargetMinimum=-0.18
     bClamp=true
     Value=@DeformRig_GetDeltaRefTransform_FNC.OutTransform.Rotation.Yaw
    ).Result
    A=@DeformRig_GetDeltaRefTransform_FNC.OutTransform.Rotation.Yaw
   )
   Make = RigVMDispatchNode(Elements=(Pitch=5,Yaw=0,Roll=5), Elements.Pitch=@n3.Result, Elements.Roll=@n3.Result, Elements.Yaw=@Multiply.Result) {template="DISPATCH_RigVMDispatch_MakeStruct(in Elements,out Struct)"}
   RigVMFunction_MathQuaternionFromRotatorV2 = /Script/RigVM.RigVMFunction_MathQuaternionFromRotatorV2(Value=@Make.Struct) {is=struct RigVMUnitNode, template="Cast::Execute(in Value,out Result)"}
   > &RigUnit_ModifyTransforms(
    ItemToModify=((Item=(Type=Bone,Name="deform_sleeve_l"),Transform=(Rotation=(X=0,Y=0,Z=0,W=1),Translation=(X=0,Y=0,Z=0),Scale3D=(X=1,Y=1,Z=1))))
    Weight=1
    WeightMinimum=0
    ItemToModify.0.Transform.Rotation=@RigVMFunction_MathQuaternionFromRotatorV2.Result
   )
   > /Script/ControlRig.RigUnit_SetCurveValue(
    Curve=ElbowLBentInv
    Value=&RigVMFunction_MathFloatRemap(
     SourceMaximum=80
     SourceMinimum=90
     TargetMaximum=0
     TargetMinimum=1
     bClamp=true
     Value=@DeformRig_GetDeltaRefTransform_FNC.OutTransform.Rotation.Yaw
    ).Result
   ) {is=struct RigVMUnitNode}
   DeformRig_GetDeltaRefTransform_FNC_1 = RigVMFunctionReferenceNode(Item=(Type=Bone,Name="lowerarm_r"), ExecuteContext=@RigVMFunction_Sequence_1.B)
   n4 = &RigVMFunction_MathFloatRemap(
    SourceMaximum=110
    SourceMinimum=70
    TargetMaximum=5
    TargetMinimum=0
    bClamp=true
    Value=@DeformRig_GetDeltaRefTransform_FNC_1.OutTransform.Rotation.Yaw
   )
   n5 = &RigVMFunction_MathDoubleMul(
    B=&RigVMFunction_MathFloatRemap(
     SourceMaximum=110
     SourceMinimum=70
     TargetMaximum=-0.08
     TargetMinimum=-0.18
     bClamp=true
     Value=@DeformRig_GetDeltaRefTransform_FNC_1.OutTransform.Rotation.Yaw
    ).Result
    A=@DeformRig_GetDeltaRefTransform_FNC_1.OutTransform.Rotation.Yaw
   )
   > &RigUnit_ModifyTransforms(
    ItemToModify=((Item=(Type=Bone,Name="deform_sleeve_r"),Transform=(Rotation=(X=0,Y=0,Z=0,W=1),Translation=(X=0,Y=0,Z=0),Scale3D=(X=1,Y=1,Z=1))))
    Weight=1
    WeightMinimum=0
    ItemToModify.0.Transform.Rotation=&RigVMFunction_MathQuaternionFromRotatorV2(Value=RigVMDispatchNode(Elements=(Pitch=0,Yaw=0,Roll=0), Elements.Pitch=@n4.Result, Elements.Roll=@n4.Result, Elements.Yaw=@n5.Result) {template="DISPATCH_RigVMDispatch_MakeStruct(in Elements,out Struct)"}.Struct) {template="Cast::Execute(in Value,out Result)"}.Result
   )
   > &RigUnit_SetCurveValue(
    Curve=ElbowRBentInv
    Value=&RigVMFunction_MathFloatRemap(
     SourceMaximum=80
     SourceMinimum=90
     TargetMaximum=0
     TargetMinimum=1
     bClamp=true
     Value=@DeformRig_GetDeltaRefTransform_FNC_1.OutTransform.Rotation.Yaw
    ).Result
   )
   n6 = &RigVMFunction_ControlFlowBranch(ExecuteContext=@RigVMFunction_Sequence_1.C, Condition=RigVMVariableNode() {variable=isLoFi}.Value)
   &RigUnit_ModifyTransforms(
    ItemToModify=((Item=(Type=Bone,Name="deform_sleeve_r"),Transform=(Rotation=(X=0,Y=0,Z=0,W=1),Translation=(X=0,Y=-0.5,Z=0),Scale3D=(X=1,Y=1,Z=1))))
    WeightMinimum=0
    Weight=/Script/ControlRig.RigUnit_GetCurveValue(Curve=ElbowRBentInv) {is=struct RigVMUnitNode}.Value
    ExecutePin=@n6.False
   )
   > &RigUnit_ModifyTransforms(
    ItemToModify=((Item=(Type=Bone,Name="deform_sleeve_l"),Transform=(Rotation=(X=0,Y=0,Z=0,W=1),Translation=(X=0,Y=0.5,Z=0),Scale3D=(X=1,Y=1,Z=1))))
    WeightMinimum=0
    Weight=&RigUnit_GetCurveValue(Curve=ElbowLBentInv).Value
   )
   n7 = &RigUnit_GetCurveValue(Curve=ElbowRBentInv)
   RigVMFunction_MathIntToDouble = /Script/RigVM.RigVMFunction_MathIntToFloat(Value=RigVMVariableNode() {variable=LOD}.Value) {is=struct RigVMUnitNode, template="Cast::Execute(in Value,out Result)"}
   n8 = &RigVMFunction_MathDoubleMul(A=3, B=@RigVMFunction_MathIntToDouble.Result)
   &RigUnit_ModifyTransforms(
    ItemToModify=((Item=(Type=Bone,Name="dyn_sleeveBk_r_01"),Transform=(Rotation=(X=0,Y=0,Z=0,W=1),Translation=(X=3,Y=0,Z=0),Scale3D=(X=1.7,Y=1,Z=1))))
    WeightMinimum=0
    Weight=@n7.Value
    ExecutePin=@n6.True
    ItemToModify.0.Transform.Translation.X=@n8.Result
   )
   > &RigUnit_ModifyTransforms(
    ItemToModify="((Item=(Type=Bone,Name=\"deform_sleeve_r\"),Transform=(Rotation=(X=0,Y=0,Z=0,W=1),Translation=(X=0,Y=-1,Z=0),Scale3D=(X=1,Y=1,Z=1))),(Item=(Type=Bone,Name=\"deform_upperarm_bk_r\"),Transform=(Rotation=(X=0,Y=0,Z=0,W=1),Translation=(X=0,Y=-1.5,Z=0),Scale3D=(X=1 ...(truncated)"
    WeightMinimum=0
    Weight=@n7.Value
   )
   > &RigUnit_ModifyTransforms(
    ItemToModify=((Item=(Type=Bone,Name="deform_sleeve_l"),Transform=(Rotation=(X=0,Y=0,Z=0,W=1),Translation=(X=0,Y=1,Z=0),Scale3D=(X=1.5,Y=1,Z=1))))
    WeightMinimum=0
    Weight=&RigUnit_GetCurveValue(Curve=ElbowLBentInv).Value
    ItemToModify.0.Transform.Translation.X=&RigVMFunction_MathDoubleMul(B=-1, A=@n8.Result).Result
   )
   n9 = &RigUnit_SphericalPoseReader(
    ActiveRegionScaleFactors=(PositiveWidth=1,NegativeWidth=1,PositiveHeight=1,NegativeHeight=1)
    ActiveRegionSize=0.1
    Debug=(bDrawDebug=True,bDraw2D=False,bDrawLocalAxes=False,DebugScale=25,DebugSegments=20,DebugThickness=0.25)
    DriverAxis=(X=-1,Y=0,Z=0)
    DriverItem=(Type=Bone,Name="upperarm_r")
    FalloffRegionScaleFactors=(PositiveWidth=1,NegativeWidth=1,PositiveHeight=1,NegativeHeight=1)
    FalloffSize=0.2
    FlipHeightScaling=False
    FlipWidthScaling=False
    OptionalParentItem=(Type=Bone,Name="spine_05")
    RotationOffset=(X=0,Y=45,Z=0)
    ExecutePin=@RigVMFunction_Sequence.B
   )
   > &RigUnit_ModifyTransforms(
    ItemToModify=((Item=(Type=Bone,Name="dyn_sleeveIn_r_master"),Transform=(Rotation=(X=0,Y=0,Z=0,W=1),Translation=(X=1,Y=0,Z=-2),Scale3D=(X=1,Y=1,Z=1))))
    WeightMinimum=0
    Weight=@n9.OutputParam
   )
   > n10 = &RigUnit_SphericalPoseReader(
    ActiveRegionScaleFactors=(PositiveWidth=1,NegativeWidth=1,PositiveHeight=1,NegativeHeight=1)
    ActiveRegionSize=0.1
    Debug=(bDrawDebug=True,bDraw2D=False,bDrawLocalAxes=False,DebugScale=25,DebugSegments=20,DebugThickness=0.25)
    DriverAxis=(X=1,Y=0,Z=0)
    DriverItem=(Type=Bone,Name="upperarm_l")
    FalloffRegionScaleFactors=(PositiveWidth=1,NegativeWidth=1,PositiveHeight=1,NegativeHeight=1)
    FalloffSize=0.2
    FlipHeightScaling=False
    FlipWidthScaling=False
    OptionalParentItem=(Type=Bone,Name="spine_05")
    RotationOffset=(X=0,Y=-130,Z=0)
   )
   > &RigUnit_ModifyTransforms(
    ItemToModify=((Item=(Type=Bone,Name="dyn_sleeveIn_l_master"),Transform=(Rotation=(X=0,Y=0,Z=0,W=1),Translation=(X=-1,Y=0,Z=2),Scale3D=(X=1,Y=1,Z=1))))
    WeightMinimum=0
    Weight=@n10.OutputParam
   )
   > n11 = &RigUnit_SphericalPoseReader(
    ActiveRegionScaleFactors=(PositiveWidth=1.5,NegativeWidth=1,PositiveHeight=1,NegativeHeight=1)
    ActiveRegionSize=0.1
    Debug=(bDrawDebug=True,bDraw2D=False,bDrawLocalAxes=False,DebugScale=25,DebugSegments=20,DebugThickness=0.25)
    DriverAxis=(X=1,Y=0,Z=0)
    DriverItem=(Type=Bone,Name="calf_r")
    FalloffRegionScaleFactors=(PositiveWidth=2,NegativeWidth=1,PositiveHeight=1,NegativeHeight=1)
    FalloffSize=0.2
    FlipHeightScaling=False
    FlipWidthScaling=False
    OptionalParentItem=(Type=Bone,Name="spine_05")
    RotationOffset=(X=90,Y=0,Z=-45)
   )
   > &RigUnit_ModifyTransforms(
    ItemToModify=((Item=(Type=Bone,Name="dyn_pant_r_master"),Transform=(Rotation=(X=0,Y=-0.0698,Z=0,W=0.9976),Translation=(X=0,Y=0,Z=0),Scale3D=(X=1,Y=1,Z=1))))
    WeightMinimum=0
    Weight=@n11.OutputParam
   )
   > n12 = &RigUnit_SphericalPoseReader(
    ActiveRegionScaleFactors=(PositiveWidth=1.5,NegativeWidth=1,PositiveHeight=1,NegativeHeight=1)
    ActiveRegionSize=0.1
    Debug=(bDrawDebug=True,bDraw2D=False,bDrawLocalAxes=False,DebugScale=25,DebugSegments=20,DebugThickness=0.25)
    DriverAxis=(X=-1,Y=0,Z=0)
    DriverItem=(Type=Bone,Name="calf_l")
    FalloffRegionScaleFactors=(PositiveWidth=2,NegativeWidth=1,PositiveHeight=1,NegativeHeight=1)
    FalloffSize=0.2
    FlipHeightScaling=False
    FlipWidthScaling=False
    OptionalParentItem=(Type=Bone,Name="spine_05")
    RotationOffset=(X=-110,Y=-170,Z=-70)
   )
   > &RigUnit_ModifyTransforms(
    ItemToModify=((Item=(Type=Bone,Name="dyn_pant_l_master"),Transform=(Rotation=(X=0,Y=-0.0698,Z=0,W=0.9976),Translation=(X=0,Y=0,Z=0),Scale3D=(X=1,Y=1,Z=1))))
    WeightMinimum=0
    Weight=@n12.OutputParam
   )
   DISPATCH_RigDispatch_GetAnimAttribute = RigVMDispatchNode(BoneName=root, Default=0, Name=deformRig_l_thigh_up_alpha) {template="DISPATCH_RigDispatch_GetAnimAttribute(in Name,in BoneName,in Default,out Value,out Found)"}
   &RigUnit_ModifyTransforms(
    ItemToModify=((Item=(Type=Bone,Name="dyn_chain_frnt_base"),Transform=(Rotation=(X=0.0508,Y=-0.1994,Z=0.5667,W=0.7978),Translation=(X=-1,Y=0,Z=0),Scale3D=(X=1,Y=1,Z=1))))
    WeightMinimum=0
    Weight=@DISPATCH_RigDispatch_GetAnimAttribute.Value
    ExecutePin=@RigVMFunction_Sequence.C
   )
   RigVMFunction_ControlFlowBranch = &RigVMFunction_ControlFlowBranch(ExecuteContext=@RigVMFunction_Sequence.D, Condition=RigVMVariableNode() {variable=is DBNO}.Value)
   &RigUnit_ModifyTransforms(
    ItemToModify=((Item=(Type=Bone,Name="dyn_hoodie_bk_master"),Transform=(Rotation=(X=0,Y=0,Z=-0.1305,W=0.9914),Translation=(X=0,Y=0,Z=0),Scale3D=(X=1,Y=1,Z=1))))
    Weight=1
    WeightMinimum=0
    ExecutePin=@RigVMFunction_ControlFlowBranch.True
   )
   n13 = &RigUnit_SphericalPoseReader(
    ActiveRegionScaleFactors=(PositiveWidth=1,NegativeWidth=1,PositiveHeight=1,NegativeHeight=1)
    ActiveRegionSize=0.1
    Debug=(bDrawDebug=True,bDraw2D=False,bDrawLocalAxes=False,DebugScale=25,DebugSegments=20,DebugThickness=0.25)
    DriverAxis=(X=-1,Y=0,Z=0)
    DriverItem=(Type=Bone,Name="upperarm_r")
    ExecutePin=()
    FalloffRegionScaleFactors=(PositiveWidth=1,NegativeWidth=1,PositiveHeight=1,NegativeHeight=1)
    FalloffSize=0.2
    FlipHeightScaling=False
    FlipWidthScaling=False
    OptionalParentItem=(Type=Bone,Name="spine_05")
    RotationOffset=(X=0,Y=45,Z=0)
   )
   > &RigUnit_ModifyTransforms(
    ItemToModify=((Item=(Type=Bone,Name="dyn_hoodie_r_master"),Transform=(Rotation=(X=0,Y=0,Z=0,W=1),Translation=(X=-1,Y=0,Z=1),Scale3D=(X=1,Y=1,Z=1))))
    WeightMinimum=0
    Weight=@n13.OutputParam
   )
   > n14 = &RigUnit_SphericalPoseReader(
    ActiveRegionScaleFactors=(PositiveWidth=1,NegativeWidth=1,PositiveHeight=1,NegativeHeight=1)
    ActiveRegionSize=0.1
    Debug=(bDrawDebug=True,bDraw2D=False,bDrawLocalAxes=False,DebugScale=25,DebugSegments=20,DebugThickness=0.25)
    DriverAxis=(X=1,Y=0,Z=0)
    DriverItem=(Type=Bone,Name="upperarm_l")
    FalloffRegionScaleFactors=(PositiveWidth=1,NegativeWidth=1,PositiveHeight=1,NegativeHeight=1)
    FalloffSize=0.2
    FlipHeightScaling=False
    FlipWidthScaling=False
    OptionalParentItem=(Type=Bone,Name="spine_05")
    RotationOffset=(X=0,Y=-130,Z=0)
   )
   > &RigUnit_ModifyTransforms(
    ItemToModify=((Item=(Type=Bone,Name="dyn_hoodie_l_master"),Transform=(Rotation=(X=0,Y=0,Z=0,W=1),Translation=(X=-1,Y=0,Z=-1),Scale3D=(X=1,Y=1,Z=1))))
    WeightMinimum=0
    Weight=@n14.OutputParam
   )
  ]}
]
hierarchy [
 root {loc=<0 0 0>, rot=<0 0 0 1>}
 pelvis {parent=root, loc=<0 -0.5301 98.0571>, rot=<0.0127 0.707 -0.0127 -0.707>}
 spine_01 {parent=pelvis, loc=<6.1208 -0.9328 0>, rot=<0 0 0.034 -0.9994>}
 spine_02 {parent=spine_01, loc=<6.5997 -0.2095 0>, rot=<0 0 0.0238 -0.9997>}
 spine_03 {parent=spine_02, loc=<6.4914 -0.1237 0>, rot=<0 0 0.0149 -0.9999>}
 spine_04 {parent=spine_03, loc=<6.9598 0.0794 0>, rot=<0 0 -0.0132 -0.9999>}
 spine_05 {parent=spine_04, loc=<6.9987 0 0>, rot=<0 0 -0.0015 -1>}
 neck_01 {parent=spine_05, loc=<11.3138 0.1529 0>, rot=<0 0 -0.2003 -0.9797>}
 neck_02 {parent=neck_01, loc=<4.5627 0 0>, rot=<0 0 0.0504 -0.9987>}
 head {parent=neck_02, loc=<5.3914 0 0>, rot=<0 0 0.115 -0.9934>}
 clavicle_l {parent=spine_05, loc=<5.4831 -0.1053 -3.2946>, rot=<0.0465 -0.7289 -0.009 -0.6829>}
 upperarm_l {parent=clavicle_l, loc=<9.8658 -1.2136 -0.4105>, rot=<0.0164 -0.3245 -0.0269 -0.9454>}
 lowerarm_l {parent=upperarm_l, loc=<26.465 -0.8345 0.259>, rot=<0.0036 0.0161 -0.2298 -0.9731>}
 hand_l {parent=lowerarm_l, loc=<22.7355 -0.3203 -0.6105>, rot=<-0.6784 -0.0223 0.0735 -0.7307>}
 hand_attach_l {parent=hand_l, loc=<8.5 -3 0>, rot=<0.7071 0.7071 0 0>}
 weapon_l {parent=hand_l, loc=<-0.1974 -3.4111 0.3811>, rot=<0 0 0 1>}
 thumb_01_l {parent=hand_l, loc=<1.3118 -0.3191 -2.459>, rot=<0.6546 -0.4042 0.0664 -0.6354>}
 thumb_02_l {parent=thumb_01_l, loc=<5.9336 -0.0448 0.0401>, rot=<-0.0037 0.0213 0.298 -0.9543>}
 thumb_03_l {parent=thumb_02_l, loc=<2.2223 0.3682 -0.0907>, rot=<0.0118 -0.013 0.0066 -0.9998>}
 index_metacarpal_l {parent=hand_l, loc=<1.6535 0.0183 -2.96>, rot=<0.0158 -0.0124 -0.0173 -0.9996>}
 index_01_l {parent=index_metacarpal_l, loc=<8.3008 0.0357 -0.0524>, rot=<-0.0044 0.0485 0.0924 -0.9945>}
 index_02_l {parent=index_01_l, loc=<3.3587 -0.3004 0.0452>, rot=<-0.002 0.0027 0.3225 -0.9466>}
 index_03_l {parent=index_02_l, loc=<2.5578 -0.123 -0.0252>, rot=<-0.002 0.0067 0.2558 -0.9667>}
 middle_metacarpal_l {parent=hand_l, loc=<2.1856 -0.1725 -0.5984>, rot=<0.006 0.0168 -0.0329 -0.9993>}
 middle_01_l {parent=middle_metacarpal_l, loc=<7.3275 -0.2447 0.0129>, rot=<-0.0208 0.0042 0.0858 -0.9961>}
 middle_02_l {parent=middle_01_l, loc=<4.0123 0.1401 0.1175>, rot=<-0.0218 0.0258 0.3512 -0.9357>}
 middle_03_l {parent=middle_02_l, loc=<2.6077 0.0534 0.1373>, rot=<0.0082 0.0188 0.1708 -0.9851>}
 ring_metacarpal_l {parent=hand_l, loc=<2.2074 -0.3625 1.3638>, rot=<-0.0284 0.0722 -0.0458 -0.9959>}
 ring_01_l {parent=ring_metacarpal_l, loc=<7.0295 -0.6197 -0.1446>, rot=<-0.0489 -0.0252 0.1892 -0.9804>}
 ring_02_l {parent=ring_01_l, loc=<3.6174 0.4647 -0.1027>, rot=<0 0 0.2806 -0.9598>}
 ring_03_l {parent=ring_02_l, loc=<2.4152 0.114 0.0199>, rot=<0 0 0.1189 -0.9929>}
 pinky_metacarpal_l {parent=hand_l, loc=<1.7616 -0.9753 3.3814>, rot=<-0.0397 0.0814 -0.069 -0.9935>}
 pinky_01_l {parent=pinky_metacarpal_l, loc=<7.1709 -0.4831 -0.1094>, rot=<-0.0654 -0.0035 0.2242 -0.9723>}
 pinky_02_l {parent=pinky_01_l, loc=<2.8274 0.0893 -0.1043>, rot=<0.0037 0.0011 0.2854 -0.9584>}
 pinky_03_l {parent=pinky_02_l, loc=<2.2046 0.0307 -0.0078>, rot=<-0.0002 0.0098 0.1445 -0.9895>}
 deform_wrist_dn_base_l {parent=hand_l, loc=<0 -1.5326 0>, rot=<0 0 0 -1>}
 deform_wrist_dn_l {parent=deform_wrist_dn_base_l, loc=<0 0 0>, rot=<0 0 0 1>}
 deform_wrist_up_base_l {parent=hand_l, loc=<0 2.3009 0>, rot=<0 0 0 -1>}
 deform_wrist_up_l {parent=deform_wrist_up_base_l, loc=<0 0 0>, rot=<0 0 0 1>}
 lowerarm_twist_02_l {parent=lowerarm_l, loc=<10.6707 0 0>, rot=<0 0 0 1>}
 lowerarm_twist_01_l {parent=lowerarm_l, loc=<18.316 0 0>, rot=<0 0 0 1>}
 deform_elbow_in_base_l {parent=lowerarm_l, loc=<0.1834 3.6062 -0.2572>, rot=<0 0 0 -1>}
 deform_elbow_in_l {parent=deform_elbow_in_base_l, loc=<0 0 0>, rot=<0 0 0 1>}
 deform_elbow_out_base_l {parent=lowerarm_l, loc=<0 -4 0>, rot=<0 0 0 1>}
 deform_elbow_out_l {parent=deform_elbow_out_base_l, loc=<0 0 0>, rot=<0 0 0 1>}
 upperarm_twist_01_l {parent=upperarm_l, loc=<7.2135 -0.1134 0>, rot=<0 0 0 1>}
 deform_sleeve_l {parent=upperarm_twist_01_l, loc=<0 0 0>, rot=<0 0 0 1>}
 dyn_sleeveFr_l_master {parent=deform_sleeve_l, loc=<5.2117 6.6159 0.7981>, rot=<0.0013 0.0165 -0.0903 -0.9958>}
 dyn_sleeveFr_l_01 {parent=dyn_sleeveFr_l_master, loc=<0 0 0>, rot=<0 0 0 -1>}
 dyn_sleeveIn_l_master {parent=deform_sleeve_l, loc=<6.372 0.5617 -7.3362>, rot=<0.0194 -0.0563 -0.0332 -0.9977>}
 dyn_sleeveIn_l_01 {parent=dyn_sleeveIn_l_master, loc=<0 0 0>, rot=<0 0 0 1>}
 dyn_sleeveOut_l_master {parent=deform_sleeve_l, loc=<5.8493 0.1188 5.2829>, rot=<0.0247 0.1543 0.0046 -0.9877>}
 dyn_sleeveOut_l_01 {parent=dyn_sleeveOut_l_master, loc=<0 0 0>, rot=<0 0 0 1>}
 dyn_sleeveBk_l_master {parent=deform_sleeve_l, loc=<6.2297 -6.3259 0.2652>, rot=<-0.0829 0.0057 0.1009 -0.9914>}
 dyn_sleeveBk_l_01 {parent=dyn_sleeveBk_l_master, loc=<0 0 0>, rot=<0 0 0 1>}
 deform_upperarm_bk_base_l {parent=upperarm_twist_01_l, loc=<-6.4053 -4.3746 0.314>, rot=<0 0 0 -1>}
 deform_upperarm_bk_l {parent=deform_upperarm_bk_base_l, loc=<0 0 0>, rot=<0 0 0 1>}
 deform_upperarm_fr_base_l {parent=upperarm_twist_01_l, loc=<-5.6312 5.092 0.3354>, rot=<0 0 0 -1>}
 deform_upperarm_fr_l {parent=deform_upperarm_fr_base_l, loc=<0 0 0>, rot=<0 0 0 1>}
 deform_upperarm_up_base_l {parent=upperarm_twist_01_l, loc=<-8.891 0.4825 3.9517>, rot=<0 0 0 -1>}
 deform_upperarm_up_l {parent=deform_upperarm_up_base_l, loc=<0 0 0>, rot=<0 0 0 1>}
 upperarm_twist_02_l {parent=upperarm_l, loc=<19.7998 -0.1134 0>, rot=<0 0 0 1>}
 clavicle_r {parent=spine_05, loc=<5.4831 -0.1053 3.2946>, rot=<0.7289 0.0465 -0.6829 0.009>}
 upperarm_r {parent=clavicle_r, loc=<-9.8658 1.2136 0.4105>, rot=<0.0164 -0.3245 -0.0269 -0.9454>}
 lowerarm_r {parent=upperarm_r, loc=<-26.465 0.8345 -0.259>, rot=<0.0036 0.0161 -0.2298 -0.9731>}
 hand_r {parent=lowerarm_r, loc=<-22.7355 0.3203 0.6106>, rot=<-0.6784 -0.0223 0.0735 -0.7307>}
 thumb_01_r {parent=hand_r, loc=<-1.3118 0.3191 2.459>, rot=<0.6546 -0.4042 0.0664 -0.6354>}
 thumb_02_r {parent=thumb_01_r, loc=<-5.9336 0.0448 -0.0401>, rot=<-0.0037 0.0213 0.298 -0.9543>}
 thumb_03_r {parent=thumb_02_r, loc=<-2.2223 -0.3682 0.0907>, rot=<0.0118 -0.013 0.0066 -0.9998>}
 index_metacarpal_r {parent=hand_r, loc=<-1.6535 -0.0183 2.96>, rot=<0.0158 -0.0124 -0.0173 -0.9996>}
 index_01_r {parent=index_metacarpal_r, loc=<-8.3008 -0.0357 0.0524>, rot=<-0.0044 0.0485 0.0924 -0.9945>}
 index_02_r {parent=index_01_r, loc=<-3.3587 0.3004 -0.0452>, rot=<-0.002 0.0027 0.3225 -0.9466>}
 index_03_r {parent=index_02_r, loc=<-2.5578 0.123 0.0252>, rot=<-0.002 0.0067 0.2558 -0.9667>}
 middle_metacarpal_r {parent=hand_r, loc=<-2.1856 0.1725 0.5984>, rot=<0.006 0.0168 -0.0329 -0.9993>}
 middle_01_r {parent=middle_metacarpal_r, loc=<-7.3275 0.2447 -0.0129>, rot=<-0.0208 0.0042 0.0858 -0.9961>}
 middle_02_r {parent=middle_01_r, loc=<-4.0123 -0.1401 -0.1175>, rot=<-0.0218 0.0258 0.3512 -0.9357>}
 middle_03_r {parent=middle_02_r, loc=<-2.6077 -0.0534 -0.1373>, rot=<0.0082 0.0188 0.1708 -0.9851>}
 ring_metacarpal_r {parent=hand_r, loc=<-2.2074 0.3625 -1.3638>, rot=<-0.0284 0.0722 -0.0458 -0.9959>}
 ring_01_r {parent=ring_metacarpal_r, loc=<-7.0295 0.6197 0.1446>, rot=<-0.0489 -0.0252 0.1892 -0.9804>}
 ring_02_r {parent=ring_01_r, loc=<-3.6174 -0.4647 0.1027>, rot=<0 0 0.2806 -0.9598>}
 ring_03_r {parent=ring_02_r, loc=<-2.4152 -0.114 -0.0199>, rot=<0 0 0.1189 -0.9929>}
 pinky_metacarpal_r {parent=hand_r, loc=<-1.7616 0.9753 -3.3814>, rot=<-0.0397 0.0814 -0.069 -0.9935>}
 pinky_01_r {parent=pinky_metacarpal_r, loc=<-7.1709 0.4831 0.1094>, rot=<-0.0654 -0.0035 0.2242 -0.9723>}
 pinky_02_r {parent=pinky_01_r, loc=<-2.8274 -0.0893 0.1043>, rot=<0.0037 0.0011 0.2854 -0.9584>}
 pinky_03_r {parent=pinky_02_r, loc=<-2.2046 -0.0307 0.0078>, rot=<-0.0002 0.0098 0.1445 -0.9895>}
 weapon_r {parent=hand_r, loc=<0.1974 3.4111 -0.3811>, rot=<0 0 0 1>}
 hand_attach_r {parent=hand_r, loc=<-8.5001 2.9999 0>, rot=<0 0 -0.7071 -0.7071>}
 deform_wrist_dn_base_r {parent=hand_r, loc=<-0.0001 1.5328 0>, rot=<0 0 0 -1>}
 deform_wrist_dn_r {parent=deform_wrist_dn_base_r, loc=<0 0 0>, rot=<0 0 0 1>}
 deform_wrist_up_base_r {parent=hand_r, loc=<-0.0002 -2.3007 -0.0001>, rot=<0 0 0 -1>}
 deform_wrist_up_r {parent=deform_wrist_up_base_r, loc=<0 0 0>, rot=<0 0 0 1>}
 lowerarm_twist_02_r {parent=lowerarm_r, loc=<-10.6707 0 0>, rot=<0 0 0 -1>}
 lowerarm_twist_01_r {parent=lowerarm_r, loc=<-18.316 0 0>, rot=<0 0 0 -1>}
 deform_elbow_in_base_r {parent=lowerarm_r, loc=<-0.1834 -3.6062 0.2572>, rot=<0 0 0 -1>}
 deform_elbow_in_r {parent=deform_elbow_in_base_r, loc=<0 0 0>, rot=<0 0 0 1>}
 deform_elbow_out_base_r {parent=lowerarm_r, loc=<0 4 -0.0001>, rot=<0 0 0 -1>}
 deform_elbow_out_r {parent=deform_elbow_out_base_r, loc=<0 0 0>, rot=<0 0 0 1>}
 upperarm_twist_01_r {parent=upperarm_r, loc=<-7.2135 0.1134 0>, rot=<0 0 0 -1>}
 deform_sleeve_r {parent=upperarm_twist_01_r, loc=<0 0 0>, rot=<0 0 0 1>}
 dyn_sleeveFr_r_master {parent=deform_sleeve_r, loc=<-5.2117 -6.6159 -0.7981>, rot=<0.0813 0.0759 -0.1019 -0.9886>}
 dyn_sleeveFr_r_01 {parent=dyn_sleeveFr_r_master, loc=<0 0 0>, rot=<0 0 0 -1>}
 dyn_sleeveBk_r_master {parent=deform_sleeve_r, loc=<-6.2298 6.3259 -0.2651>, rot=<0.0058 0.0656 0.0929 -0.9935>}
 dyn_sleeveBk_r_01 {parent=dyn_sleeveBk_r_master, loc=<0 0 0>, rot=<0 0 0 1>}
 dyn_sleeveOut_r_master {parent=deform_sleeve_r, loc=<-5.849 -0.1188 -5.2832>, rot=<0.1366 0.1962 0.002 -0.971>}
 dyn_sleeveOut_r_01 {parent=dyn_sleeveOut_r_master, loc=<0 0 0>, rot=<0 0 0 1>}
 dyn_sleeveIn_r_master {parent=deform_sleeve_r, loc=<-6.3719 -0.5617 7.3362>, rot=<0.0866 -0.0037 -0.0521 -0.9949>}
 dyn_sleeveIn_r_01 {parent=dyn_sleeveIn_r_master, loc=<0 0 0>, rot=<0 0 0 1>}
 deform_upperarm_bk_base_r {parent=upperarm_twist_01_r, loc=<6.4053 4.3746 -0.314>, rot=<0 0 0 -1>}
 deform_upperarm_bk_r {parent=deform_upperarm_bk_base_r, loc=<0 0 0>, rot=<0 0 0 1>}
 deform_upperarm_fr_base_r {parent=upperarm_twist_01_r, loc=<5.6312 -5.092 -0.3354>, rot=<0 0 0 -1>}
 deform_upperarm_fr_r {parent=deform_upperarm_fr_base_r, loc=<0 0 0>, rot=<0 0 0 1>}
 deform_upperarm_up_base_r {parent=upperarm_twist_01_r, loc=<8.891 -0.4825 -3.9517>, rot=<0 0 0 -1>}
 deform_upperarm_up_r {parent=deform_upperarm_up_base_r, loc=<0 0 0>, rot=<0 0 0 1>}
 upperarm_twist_02_r {parent=upperarm_r, loc=<-19.7998 0.1134 0>, rot=<0 0 0 -1>}
 attach_backpack {parent=spine_05, loc=<-0.123 -13.473 0>, rot=<-0.0093 -0.707 0.0093 -0.707>}
 attach_cape {parent=spine_05, loc=<9.017 -11.714 0>, rot=<0.0616 -0.7044 -0.0616 -0.7044>}
 deform_pec_aim_base_l {parent=spine_04, loc=<3.2456 9.3309 -1.2237>, rot=<0.0603 -0.4033 0.3318 -0.8507>}
 deform_pec_aim_l {parent=deform_pec_aim_base_l, loc=<0 0 0>, rot=<0 0 0 1>}
 deform_pec_base_l {parent=deform_pec_aim_l, loc=<9.3665 0 0>, rot=<0 0 0 1>}
 deform_pec_l {parent=deform_pec_base_l, loc=<0 0 0>, rot=<0 0 0 1>}
 deform_pec_aim_base_r {parent=spine_04, loc=<3.2452 9.3309 1.2237>, rot=<-0.4033 -0.0603 0.8507 0.3318>}
 deform_pec_aim_r {parent=deform_pec_aim_base_r, loc=<0 0 0>, rot=<0 0 0 1>}
 deform_pec_base_r {parent=deform_pec_aim_r, loc=<-9.3667 -0.0003 -0.0004>, rot=<0 0 0 1>}
 deform_pec_r {parent=deform_pec_base_r, loc=<0 0 0>, rot=<0 0 0 1>}
 dyn_hoodie_l_master {parent=spine_03, loc=<0.9875 -1.3738 -11.785>, rot=<0.0671 -0.9939 -0.0059 -0.0877>}
 dyn_hoodie_l {parent=dyn_hoodie_l_master, loc=<0 0 0>, rot=<0 0 0 1>}
 dyn_hoodie_bk_master {parent=spine_03, loc=<2.8733 -10.3483 0>, rot=<0 0 0.9958 -0.0912>}
 dyn_hoodie_bk {parent=dyn_hoodie_bk_master, loc=<0 0 0>, rot=<0 0 0 1>}
 dyn_hoodie_fr_master {parent=spine_03, loc=<1.2013 12.0332 0>, rot=<-0.1041 -0.9946 0 0>}
 dyn_hoodie_fr {parent=dyn_hoodie_fr_master, loc=<0 0 0>, rot=<0 0 0 1>}
 dyn_hoodie_r_master {parent=spine_03, loc=<0.9875 -1.3738 11.7848>, rot=<0.0671 -0.9939 0.0059 0.0877>}
 dyn_hoodie_r {parent=dyn_hoodie_r_master, loc=<0 0 0>, rot=<0 0 0 1>}
 thigh_l {parent=pelvis, loc=<-1.0855 -0.2109 -8.7053>, rot=<0.0009 0.0355 0.0155 -0.9993>}
 calf_l {parent=thigh_l, loc=<-42.5284 -0.0019 -0.7297>, rot=<0.0006 -0.0161 -0.0391 -0.9991>}
 foot_l {parent=calf_l, loc=<-46.7401 -2.5014 -0.592>, rot=<0.0427 -0.0238 0.0353 -0.9982>}
 ball_l {parent=foot_l, loc=<-3.5658 10.8066 -0.1544>, rot=<0.0001 -0.0001 -0.7063 -0.7079>}
 calf_twist_02_l {parent=calf_l, loc=<-19.1546 -1.7943 -0.2638>, rot=<0.0237 0.0095 -0.0035 -0.9997>}
 calf_twist_01_l {parent=calf_l, loc=<-35.8483 -1.9197 -0.5794>, rot=<0.0237 0.0095 -0.0035 -0.9997>}
 deform_foot_bk_l {parent=calf_twist_01_l, loc=<-5.893 -3.101 0.3778>, rot=<0.6744 -0.7373 -0.0393 0.003>}
 deform_knee_in_base_l {parent=calf_l, loc=<0.8031 -5.0231 0.2456>, rot=<0 0 0 -1>}
 deform_knee_in_l {parent=deform_knee_in_base_l, loc=<0 0 0>, rot=<0 0 0 1>}
 deform_knee_out_base_l {parent=calf_l, loc=<0 4 0>, rot=<0 0 0 -1>}
 deform_knee_out_l {parent=deform_knee_out_base_l, loc=<0 0 0>, rot=<0 0 0 1>}
 dyn_pant_l_master {parent=calf_l, loc=<-16.7343 -0.8956 -0.2119>, rot=<-0.0226 0.6704 0.7415 -0.0156>}
 dyn_pant_l {parent=dyn_pant_l_master, loc=<0 0 0>, rot=<0 0 0 1>}
 thigh_twist_01_l {parent=thigh_l, loc=<-27.952 0 0>, rot=<0.0013 0 0 -1>}
 deform_groin_base_l {parent=thigh_l, loc=<0.4801 5.6776 -0.2486>, rot=<0 0 0 -1>}
 deform_groin_l {parent=deform_groin_base_l, loc=<0 0 0>, rot=<0 0 0 1>}
 thigh_r {parent=pelvis, loc=<-1.0855 -0.2109 8.7053>, rot=<0.0355 -0.0009 0.9993 0.0155>}
 calf_r {parent=thigh_r, loc=<42.5284 0.0019 0.7297>, rot=<0.0006 -0.0161 -0.0391 -0.9991>}
 foot_r {parent=calf_r, loc=<46.7401 2.5014 0.592>, rot=<0.0427 -0.0238 0.0353 -0.9982>}
 ball_r {parent=foot_r, loc=<3.5658 -10.8066 0.1544>, rot=<0.0001 -0.0001 -0.7063 -0.7079>}
 calf_twist_02_r {parent=calf_r, loc=<19.1546 1.7943 0.2638>, rot=<0.0237 0.0095 -0.0035 -0.9997>}
 calf_twist_01_r {parent=calf_r, loc=<35.8483 1.9197 0.5794>, rot=<0.0237 0.0095 -0.0035 -0.9997>}
 deform_foot_bk_r {parent=calf_twist_01_r, loc=<5.917 3.099 -0.3781>, rot=<0.7373 0.6744 -0.003 -0.0393>}
 deform_knee_in_base_r {parent=calf_r, loc=<-0.8031 5.0231 -0.2457>, rot=<0 0 0 -1>}
 deform_knee_in_r {parent=deform_knee_in_base_r, loc=<0 0 0>, rot=<0 0 0 1>}
 deform_knee_out_base_r {parent=calf_r, loc=<0 -4 0>, rot=<0 0 0 -1>}
 deform_knee_out_r {parent=deform_knee_out_base_r, loc=<0 0 0>, rot=<0 0 0 1>}
 dyn_pant_r_master {parent=calf_r, loc=<16.7343 0.8956 0.2119>, rot=<-0.7415 -0.0156 -0.0226 -0.6704>}
 dyn_pant_r {parent=dyn_pant_r_master, loc=<0 0 0>, rot=<0 0 0 1>}
 thigh_twist_01_r {parent=thigh_r, loc=<27.952 0 0>, rot=<0.0013 0 0 -1>}
 deform_groin_base_r {parent=thigh_r, loc=<-0.4801 -5.6776 0.2486>, rot=<0 0 0 -1>}
 deform_groin_r {parent=deform_groin_base_r, loc=<0 0 0>, rot=<0 0 0 1>}
 deform_glute_base_l {parent=pelvis, loc=<-7.3466 -9.0213 -9.1245>, rot=<0.0172 0.031 -0.4477 -0.8935>}
 deform_glute_l {parent=deform_glute_base_l, loc=<0 0 0>, rot=<0 0 0 1>}
 deform_glute_base_r {parent=pelvis, loc=<-7.3464 -9.0213 9.1245>, rot=<0.031 -0.0172 0.8935 -0.4477>}
 deform_glute_r {parent=deform_glute_base_r, loc=<0 0 0>, rot=<0 0 0 1>}
 dyn_chain_bk_base {parent=pelvis, loc=<9.4989 -7.8922 -6.9738>, rot=<0.2433 -0.9639 -0.1073 -0.0142>}
 dyn_chain_bk_01 {parent=dyn_chain_bk_base, loc=<0 0 0>, rot=<0 0 0 1>}
 dyn_chain_bk_02 {parent=dyn_chain_bk_01, loc=<3.8903 -0.0411 0.2817>, rot=<-0.0098 0.1769 -0.0347 -0.9836>}
 dyn_chain_bk_03 {parent=dyn_chain_bk_02, loc=<6.2681 -0.2545 0.0951>, rot=<0.0098 0.3133 -0.2212 -0.9235>}
 dyn_chain_bk_04 {parent=dyn_chain_bk_03, loc=<5.8209 0.6736 0.2516>, rot=<-0.0857 0.1165 -0.1356 -0.9801>}
 dyn_chain_frnt_base {parent=pelvis, loc=<6.6629 7.988 -7.5056>, rot=<-0.0746 -0.969 0.2258 -0.0666>}
 dyn_chain_frnt_01 {parent=dyn_chain_frnt_base, loc=<0 0 0>, rot=<0 0 0 1>}
 dyn_chain_frnt_02 {parent=dyn_chain_frnt_01, loc=<2.9923 0.2354 0.04>, rot=<0.0084 0.1732 -0.0331 -0.9843>}
 dyn_chain_frnt_03 {parent=dyn_chain_frnt_02, loc=<4.9181 -0.1122 0.1418>, rot=<-0.0265 0.1452 0.0468 -0.9879>}
 dyn_chain_frnt_04 {parent=dyn_chain_frnt_03, loc=<4.7439 -0.1788 1.015>, rot=<-0.0718 0.3613 0.2308 -0.9006>}
 dyn_chain_frnt_05 {parent=dyn_chain_frnt_04, loc=<5.7892 0.1787 -0.0755>, rot=<-0.1684 0.1133 0.1756 -0.9633>}
 chain_support_l_base {parent=pelvis, loc=<4.7872 -3.1054 -12.9314>, rot=<-0.0014 -0.9446 0.2318 -0.2325>}
 chain_support_l_01 {parent=chain_support_l_base, loc=<0 0 0>, rot=<0 0 0 1>}
 chain_support_l_02 {parent=chain_support_l_01, loc=<10 0 0>, rot=<0 -0.1826 0 -0.9832>}
 ElbowLBentInv {type=Curve, loc=<0 0 0>, rot=<0 0 0 1>}
 ElbowRBentInv {type=Curve, loc=<0 0 0>, rot=<0 0 0 1>}
]
variables [
 LOD {type=int32, default=0}
 is DBNO {type=bool, default=True}
 isLoFi {type=bool, default=False}
]
