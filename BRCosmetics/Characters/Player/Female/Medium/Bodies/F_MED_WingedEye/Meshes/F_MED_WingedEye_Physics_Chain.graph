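# Physics asset for the hanging chain on the medium female WingedEye player body, covering front, back and left support bone chains.
# Each chain links capsule bodies from a zero-size base through small swing and twist limits, with twist locked on a few joints.
# Front and back chain ends are angularly locked to the last support bone, joining the three chains.
# Capsules on pelvis, left thigh, left calf and root supply body shapes around the chain.
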PhysicsAsset /BRCosmetics/Characters/Player/Female/Medium/Bodies/F_MED_WingedEye/Meshes/F_MED_WingedEye_Physics_Chain
bodies {
 calf_l [{
   shapeType=Capsule
   length=48.3537
   radius=8.1616
   shapeName=calf_l_capsule
   center=<-26.6757 -2.6651 0>
   rotation=<rot 90 -179.8881 178.5002>
  }]
 chain_support_l_01 [{
   shapeType=Capsule
   length=11.242
   radius=2
   shapeName=chain_support_l_01_capsule
   center=<6.3245 0 0.6914>
   rotation=<rot -0.0002 -89.9997 96.3016>
  }]
 chain_support_l_02 [{
   shapeType=Capsule
   length=11.1601
   radius=4.0395
   shapeName=chain_support_l_02_capsule
   center=<1.4958 -0.6105 2.5957>
   rotation=<rot 1.1287 -10.9315 91.977>
  }]
 chain_support_l_base [{shapeType=Capsule, length=0, radius=0, shapeName=chain_support_l_base_capsule}]
 dyn_chain_bk_01 [{shapeType=Capsule, length=3.2799, radius=1.8622, shapeName=dyn_chain_bk_01_capsule, center=<2.6463 0 0>, rotation=<rot 90 0 0>}]
 dyn_chain_bk_02 [{shapeType=Capsule, length=4.0518, radius=1.8933, shapeName=dyn_chain_bk_02_capsule, center=<2.5601 -0.2917 0>, rotation=<rot 90 -2.3567 0>}]
 dyn_chain_bk_03 [{shapeType=Capsule, length=4.0518, radius=1.8933, shapeName=dyn_chain_bk_02_capsule, center=<2.5721 0 0>, rotation=<rot 90 -78.9374 -74.7542>}]
 dyn_chain_bk_04 [{shapeType=Capsule, length=4.0518, radius=1.8933, shapeName=dyn_chain_bk_02_capsule, center=<2.5884 -0.1034 0>, rotation=<rot 90 8.9239 0>}]
 dyn_chain_bk_base [{shapeType=Capsule, length=0, radius=0, shapeName=dyn_chain_bk_base_capsule}]
 dyn_chain_frnt_01 [{shapeType=Capsule, length=3.0029, radius=1.3899, shapeName=dyn_chain_frnt_01_capsule, center=<1.7007 0.3653 0>, rotation=<rot 90 0 0>}]
 dyn_chain_frnt_02 [{
   shapeType=Capsule
   length=3.1491
   radius=2.5164
   shapeName=dyn_chain_frnt_02_capsule
   center=<2.5179 0.7276 0.3027>
   rotation=<rot 90 -45.1254 -41.7546>
  }]
 dyn_chain_frnt_03 [{
   shapeType=Capsule
   length=3.1491
   radius=2.5164
   shapeName=dyn_chain_frnt_02_capsule
   center=<2.5179 0.7276 0.3027>
   rotation=<rot 90 -45.1254 -41.7546>
  }]
 dyn_chain_frnt_04 [{
   shapeType=Capsule
   length=3.1491
   radius=2.5164
   shapeName=dyn_chain_frnt_02_capsule
   center=<2.5348 1.0145 0.3027>
   rotation=<rot 90 -3.3708 0>
  }]
 dyn_chain_frnt_05 [{
   shapeType=Capsule
   length=4.6491
   radius=2.5164
   shapeName=dyn_chain_frnt_02_capsule
   center=<3.6155 0.6146 0.3027>
   rotation=<rot 90 -13.3708 0>
  }]
 dyn_chain_frnt_base [{shapeType=Capsule, length=0, radius=0, shapeName=dyn_chain_frnt_base_capsule}]
 pelvis [{shapeType=Capsule, length=6.4136, radius=12.0243, shapeName=pelvis_capsule, center=<-5.9542 -2.1131 0>}]
 root [{shapeType=Capsule, length=0, radius=73.6701, shapeName=root_capsule, center=<0 -14.1471 -71.7988>, rotation=<rot 0 -90 0>}]
 thigh_l [{
   shapeType=Capsule
   length=36.9173
   radius=8.9052
   shapeName=thigh_l_capsule
   center=<-23.0524 0.1576 1.4932>
   rotation=<rot 89.263 -162.063 -157.3827>
  }]
}
constraints [
 {
  bone1Name=chain_support_l_01
  bone2Name=chain_support_l_base
  swing1LimitDegrees=10
  swing2LimitDegrees=65
  twistLimitDegrees=20
  twistMotion=Locked
 }
 {bone1Name=chain_support_l_02, bone2Name=chain_support_l_01, swing1LimitDegrees=10, swing2LimitDegrees=10, twistLimitDegrees=10}
 {bone1Name=dyn_chain_bk_01, bone2Name=dyn_chain_bk_base, swing1LimitDegrees=10, swing2LimitDegrees=10, twistLimitDegrees=5}
 {bone1Name=dyn_chain_bk_02, bone2Name=dyn_chain_bk_01, swing1LimitDegrees=10, swing2LimitDegrees=10, twistLimitDegrees=5}
 {bone1Name=dyn_chain_bk_03, bone2Name=dyn_chain_bk_02, swing1LimitDegrees=10, swing2LimitDegrees=20, twistLimitDegrees=5}
 {bone1Name=dyn_chain_bk_04, bone2Name=dyn_chain_bk_03, swing1LimitDegrees=30, swing2LimitDegrees=30, twistLimitDegrees=5}
 {bone1Name=dyn_chain_frnt_01, bone2Name=dyn_chain_frnt_base, swing1LimitDegrees=10, swing2LimitDegrees=15, twistLimitDegrees=10}
 {bone1Name=dyn_chain_frnt_02, bone2Name=dyn_chain_frnt_01, swing1LimitDegrees=15, swing2LimitDegrees=15, twistLimitDegrees=10}
 {bone1Name=dyn_chain_frnt_03, bone2Name=dyn_chain_frnt_02, swing1LimitDegrees=30, swing2LimitDegrees=30, twistLimitDegrees=0}
 {
  bone1Name=dyn_chain_frnt_04
  bone2Name=dyn_chain_frnt_03
  swing1LimitDegrees=30
  swing2LimitDegrees=30
  twistLimitDegrees=5
  twistMotion=Locked
 }
 {bone1Name=dyn_chain_frnt_05, bone2Name=dyn_chain_frnt_04, swing1LimitDegrees=30, swing2LimitDegrees=30, twistLimitDegrees=5}
 {
  bone1Name=dyn_chain_bk_04
  bone2Name=chain_support_l_02
  swing1LimitDegrees=45
  swing1Motion=Locked
  swing2LimitDegrees=45
  swing2Motion=Locked
  twistLimitDegrees=45
  twistMotion=Locked
 }
 {
  bone1Name=dyn_chain_frnt_05
  bone2Name=chain_support_l_02
  swing1LimitDegrees=45
  swing1Motion=Locked
  swing2LimitDegrees=45
  swing2Motion=Locked
  twistLimitDegrees=45
  twistMotion=Locked
 }
]
