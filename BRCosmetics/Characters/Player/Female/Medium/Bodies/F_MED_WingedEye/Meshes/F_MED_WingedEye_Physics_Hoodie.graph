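# Physics asset for the hoodie on the WingedEye medium female player body, covering its back, front, left and right dynamic bones.
# Each hoodie bone has a capsule constrained to its own tiny-capsule master bone, with tight angular limits whose wider swing axis swaps between front/back and sides.
# Two large capsules on spine_02 form the torso volume, while spine_03 has only a tiny capsule and neither spine body is constrained.
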PhysicsAsset /BRCosmetics/Characters/Player/Female/Medium/Bodies/F_MED_WingedEye/Meshes/F_MED_WingedEye_Physics_Hoodie
bodies {
 dyn_hoodie_bk [{
   shapeType=Capsule
   length=8.0757
   radius=3.4134
   shapeName=dyn_hoodie_bk_capsule
   center=<3.3044 -0.083 -0.0457>
   rotation=<rot -0.2625 -90 2.2573>
  }]
 dyn_hoodie_bk_master [{shapeType=Capsule, length=0.505, radius=0.505, shapeName=dyn_hoodie_bk_master_capsule, rotation=<rot 0 -90 0>}]
 dyn_hoodie_fr [{shapeType=Capsule, length=8.8554, radius=3.1916, shapeName=dyn_hoodie_fr_capsule, center=<3 0.3039 0.1446>, rotation=<rot 0.4684 -90 0.7664>}]
 dyn_hoodie_fr_master [{shapeType=Capsule, length=0.505, radius=0.505, shapeName=dyn_hoodie_fr_master_capsule, rotation=<rot 0 -90 0>}]
 dyn_hoodie_l [{
   shapeType=Capsule
   length=8.399
   radius=3.1681
   shapeName=dyn_hoodie_l_capsule
   center=<4.1172 1.1401 0.5302>
   rotation=<rot 0 -9.4415 -88.7397>
  }]
 dyn_hoodie_l_master [{shapeType=Capsule, length=0.505, radius=0.505, shapeName=dyn_hoodie_l_master_capsule, rotation=<rot 0 -90 0>}]
 dyn_hoodie_r [{
   shapeType=Capsule
   length=8.2364
   radius=3.0528
   shapeName=dyn_hoodie_r_capsule
   center=<3.7563 1.51 -0.6029>
   rotation=<rot 0 173.2885 -86.3042>
  }]
 dyn_hoodie_r_master [{shapeType=Capsule, length=0.505, radius=0.505, shapeName=dyn_hoodie_r_master_capsule, rotation=<rot 0 -90 0>}]
 spine_02 [
  {shapeType=Capsule, length=5.4107, radius=6.04, shapeName=spine_02_capsule, center=<2.5971 2.1576 -0.0346>, rotation=<rot -1.3876 -90 0.6242>}
  {
   shapeType=Capsule
   length=8.0561
   radius=6.8146
   shapeName=spine_02_capsule
   center=<9.7664 2.1568 -0.0038>
   rotation=<rot -1.3876 -90 0.6242>
  }
 ]
 spine_03 [{shapeType=Capsule, length=0.505, radius=0.505, shapeName=spine_03_capsule, rotation=<rot 0 -90 0>}]
}
constraints [
 {bone1Name=dyn_hoodie_bk, bone2Name=dyn_hoodie_bk_master, swing1LimitDegrees=10, swing2LimitDegrees=5, twistLimitDegrees=5}
 {bone1Name=dyn_hoodie_fr, bone2Name=dyn_hoodie_fr_master, swing1LimitDegrees=10, swing2LimitDegrees=5, twistLimitDegrees=5}
 {bone1Name=dyn_hoodie_l, bone2Name=dyn_hoodie_l_master, swing1LimitDegrees=5, swing2LimitDegrees=10, twistLimitDegrees=5}
 {bone1Name=dyn_hoodie_r, bone2Name=dyn_hoodie_r_master, swing1LimitDegrees=5, swing2LimitDegrees=10, twistLimitDegrees=5}
]
