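# Control Rig for the WingedEye female face accessory that applies pose-driven corrections to hair and upper-arm helper bones.
# Spherical pose readers on head, clavicle and upper arms supply weights, though the right-arm helper offset stays at full weight, ignoring its reader.
# Head and right-arm readers multiplied by the Crouch flag shift four back and right-side hair base bones.
# Head, right-arm and left-clavicle readers multiply into the SPR_HeadUp curve.
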ControlRigBlueprint /BRCosmetics/Characters/Player/Female/Medium/Bodies/F_MED_WingedEye/Meshes/Parts/F_MED_WingedEye_FaceAcc_CtrlRig
graphs [
 Multiply_2_1_ContainedGraph {nodes [
   n1 = RigVMFunctionEntryNode()
   RigVMFunctionReturnNode(Result=&RigVMFunction_MathDoubleMul(B=@n1.C, A=&RigVMFunction_MathDoubleMul(A=@n1.A, B=@n1.B).Result).Result)
  ]}
 Multiply_2_1_ContainedGraph {nodes [
   n1 = RigVMFunctionEntryNode()
   RigVMFunctionReturnNode(Result=&RigVMFunction_MathDoubleMul(B=@n1.C, A=&RigVMFunction_MathDoubleMul(A=@n1.A, B=@n1.B).Result).Result)
  ]}
 RigVMFunctionLibrary {links [], nodes []}
 RigVMModel {nodes [
   RigUnit_BeginExecution = &RigUnit_BeginExecution()
   SequenceAggregate_1 = RigVMAggregateNode(ExecuteContext=@RigUnit_BeginExecution.ExecutePin)
   RigVMFunction_Sequence = &RigVMFunction_Sequence(ExecuteContext=@SequenceAggregate_1.C)
   &RigUnit_SphericalPoseReader(
    ActiveRegionScaleFactors=(PositiveWidth=1,NegativeWidth=1,PositiveHeight=1,NegativeHeight=1)
    ActiveRegionSize=0.1
    Debug=(bDrawDebug=true,bDraw2D=False,bDrawLocalAxes=False,DebugScale=25,DebugSegments=20,DebugThickness=0.25)
    DriverAxis=(X=-1,Y=0,Z=0)
    DriverItem=(Type=Bone,Name="upperarm_r")
    FalloffRegionScaleFactors=(PositiveWidth=1,NegativeWidth=1,PositiveHeight=1,NegativeHeight=1)
    FalloffSize=0.3
    OptionalParentItem=(Type=Bone,Name="None")
    RotationOffset=(X=-112,Y=4,Z=-37)
    ExecutePin=@RigVMFunction_Sequence.A
   )
   > &RigUnit_ModifyTransforms(
    ItemToModify=((Item=(Type=Bone,Name="dyn_extra_upperarm_r"),Transform=(Rotation=(X=0,Y=0,Z=0,W=1),Translation=(X=0,Y=1,Z=-5),Scale3D=(X=1,Y=1,Z=1))))
    Weight=1
   )
   n1 = &RigUnit_SphericalPoseReader(
    ActiveRegionScaleFactors=(PositiveWidth=1,NegativeWidth=1,PositiveHeight=1,NegativeHeight=1)
    ActiveRegionSize=0.1
    Debug=(bDrawDebug=true,bDraw2D=False,bDrawLocalAxes=False,DebugScale=25,DebugSegments=20,DebugThickness=0.25)
    DriverAxis=(X=1,Y=0,Z=0)
    DriverItem=(Type=Bone,Name="upperarm_l")
    FalloffRegionScaleFactors=(PositiveWidth=1,NegativeWidth=1,PositiveHeight=1,NegativeHeight=1)
    FalloffSize=0.3
    OptionalParentItem=(Type=Bone,Name="None")
    RotationOffset=(X=5,Y=-24,Z=0)
    ExecutePin=@RigVMFunction_Sequence.B
   )
   > &RigUnit_ModifyTransforms(
    ItemToModify=((Item=(Type=Bone,Name="dyn_extra_upperarm_l"),Transform=(Rotation=(X=0,Y=0,Z=0,W=1),Translation=(X=0,Y=-2,Z=0),Scale3D=(X=1,Y=1,Z=1))))
    Weight=@n1.OutputParam
   )
   n2 = &RigUnit_SphericalPoseReader(
    ActiveRegionScaleFactors=(PositiveWidth=1,NegativeWidth=1,PositiveHeight=1,NegativeHeight=1)
    ActiveRegionSize=0.1
    Debug=(bDrawDebug=true,bDraw2D=False,bDrawLocalAxes=False,DebugScale=25,DebugSegments=20,DebugThickness=0.25)
    DriverAxis=(X=-1,Y=0,Z=0)
    DriverItem=(Type=Bone,Name="upperarm_r")
    FalloffRegionScaleFactors=(PositiveWidth=1,NegativeWidth=1,PositiveHeight=1,NegativeHeight=1)
    FalloffSize=0.3
    OptionalParentItem=(Type=Bone,Name="None")
    RotationOffset=(X=-51,Y=6,Z=0)
    ExecutePin=@SequenceAggregate_1.B
   )
   > n3 = &RigUnit_SphericalPoseReader(
    ActiveRegionScaleFactors=(PositiveWidth=1,NegativeWidth=0.5,PositiveHeight=0.75,NegativeHeight=0.75)
    ActiveRegionSize=0.1
    Debug=(bDrawDebug=true,bDraw2D=False,bDrawLocalAxes=False,DebugScale=25,DebugSegments=20,DebugThickness=0.25)
    DriverAxis=(X=0,Y=1,Z=0)
    DriverItem=(Type=Bone,Name="head")
    FalloffRegionScaleFactors=(PositiveWidth=1,NegativeWidth=0.5,PositiveHeight=1,NegativeHeight=1)
    FalloffSize=0.3
    OptionalParentItem=(Type=Bone,Name="None")
    RotationOffset=(X=90,Y=0,Z=-19)
   )
   > n4 = &RigUnit_SphericalPoseReader(
    ActiveRegionScaleFactors=(PositiveWidth=1,NegativeWidth=1,PositiveHeight=1,NegativeHeight=1)
    ActiveRegionSize=0.1
    Debug=(bDrawDebug=true,bDraw2D=False,bDrawLocalAxes=False,DebugScale=25,DebugSegments=20,DebugThickness=0.25)
    DriverAxis=(X=1,Y=0,Z=0)
    DriverItem=(Type=Bone,Name="clavicle_l")
    FalloffRegionScaleFactors=(PositiveWidth=1,NegativeWidth=1,PositiveHeight=1,NegativeHeight=1)
    FalloffSize=0.2
    OptionalParentItem=(Type=Bone)
    RotationOffset=(X=39,Y=0,Z=-80)
   )
   Multiply_2_1 = RigVMAggregateNode(A=@n4.OutputParam, C=@n2.OutputParam, B=@n3.OutputParam)
   > /Script/ControlRig.RigUnit_SetCurveValue(Curve=SPR_HeadUp, Value=@Multiply_2_1.Result) {is=struct RigVMUnitNode}
   n5 = &RigUnit_SphericalPoseReader(
    ActiveRegionScaleFactors=(PositiveWidth=1,NegativeWidth=0.5,PositiveHeight=0.75,NegativeHeight=0.75)
    ActiveRegionSize=0.01
    Debug=(bDrawDebug=true,bDraw2D=False,bDrawLocalAxes=False,DebugScale=25,DebugSegments=20,DebugThickness=0.25)
    DriverAxis=(X=0,Y=1,Z=0)
    DriverItem=(Type=Bone,Name="head")
    FalloffRegionScaleFactors=(PositiveWidth=1,NegativeWidth=0.5,PositiveHeight=1,NegativeHeight=1)
    FalloffSize=0.3
    OptionalParentItem=(Type=Bone,Name="None")
    RotationOffset=(X=90,Y=0,Z=-31)
    ExecutePin=@SequenceAggregate_1.A
   )
   > n6 = &RigUnit_SphericalPoseReader(
    ActiveRegionScaleFactors=(PositiveWidth=1,NegativeWidth=1,PositiveHeight=1,NegativeHeight=1)
    ActiveRegionSize=0.01
    Debug=(bDrawDebug=true,bDraw2D=False,bDrawLocalAxes=False,DebugScale=25,DebugSegments=20,DebugThickness=0.25)
    DriverAxis=(X=-1,Y=0,Z=0)
    DriverItem=(Type=Bone,Name="upperarm_r")
    FalloffRegionScaleFactors=(PositiveWidth=1,NegativeWidth=1,PositiveHeight=1,NegativeHeight=1)
    FalloffSize=0.3
    OptionalParentItem=(Type=Bone)
    RotationOffset=(X=-105,Y=0,Z=25)
   )
   RigVMFunction_MathBoolToFloat = &RigVMFunction_MathBoolToFloat(Value=RigVMVariableNode() {variable=crouch}.Value)
   Multiply_2 = RigVMAggregateNode(C=@RigVMFunction_MathBoolToFloat.Result, A=@n6.OutputParam, B=@n5.OutputParam)
   > &RigUnit_ModifyTransforms(
    ItemToModify=((Item=(Type=Bone,Name="dyn_hair_side_bk_l_base"),Transform=(Rotation=(X=0,Y=0,Z=-0.0872,W=0.9962),Translation=(X=0,Y=-5,Z=0),Scale3D=(X=1,Y=1,Z=1))))
    Weight=@Multiply_2.Result
   )
   > &RigUnit_ModifyTransforms(
    ItemToModify=((Item=(Type=Bone,Name="dyn_hair_center_bk_base"),Transform=(Rotation=(X=0,Y=0,Z=-0.0872,W=0.9962),Translation=(X=0,Y=-5,Z=0),Scale3D=(X=1,Y=1,Z=1))))
    Weight=@Multiply_2.Result
   )
   > &RigUnit_ModifyTransforms(
    ItemToModify=((Item=(Type=Bone,Name="dyn_hair_side_bk_r_base"),Transform=(Rotation=(X=0,Y=0,Z=-0.0872,W=0.9962),Translation=(X=-5.2947,Y=-5,Z=0.4621),Scale3D=(X=1,Y=1,Z=1))))
    Weight=@Multiply_2.Result
   )
   > &RigUnit_ModifyTransforms(
    ItemToModify=((Item=(Type=Bone,Name="dyn_hair_side_r_base"),Transform=(Rotation=(X=-0.0972,Y=-0.0686,Z=0.2785,W=0.9531),Translation=(X=-2.3809,Y=-3.292,Z=-1.9035),Scale3D=(X=1,Y=1,Z=1))))
    Weight=@Multiply_2.Result
   )
  ]}
 SequenceAggregate_1_ContainedGraph {nodes [
   n1 = RigVMFunctionEntryNode()
   SequenceAggregate = &RigVMFunction_Sequence(ExecuteContext=@n1.ExecuteContext)
   SequenceAggregate_1 = &RigVMFunction_Sequence(ExecuteContext=@SequenceAggregate.B)
   SequenceAggregate_2 = &RigVMFunction_Sequence(ExecuteContext=@SequenceAggregate_1.B)
   RigVMFunctionReturnNode(A=@SequenceAggregate.A, B=@SequenceAggregate_1.A, C=@SequenceAggregate_2.A, D=@SequenceAggregate_2.B)
  ]}
]
hierarchy [
 root {loc=<0 0 0>, rot=<0 0 0 1>}
 pelvis {parent=root, loc=<0 -0.5301 98.0571>, rot=<0.0127 0.707 -0.0127 -0.707>}
 spine_01 {parent=pelvis, loc=<6.1208 -0.9328 0>, rot=<0 0 0.034 -0.9994>}
 spine_02 {parent=spine_01, loc=<6.5997 -0.2095 0>, rot=<0 0 0.0238 -0.9997>}
 spine_03 {parent=spine_02, loc=<6.4914 -0.1237 0>, rot=<0 0 0.0149 -0.9999>}
 spine_04 {parent=spine_03, loc=<6.9598 0.0794 0>, rot=<0 0 -0.0132 -0.9999>}
 spine_05 {parent=spine_04, loc=<6.9987 0 0>, rot=<0 0 -0.0015 -1>}
 neck_01 {parent=spine_05, loc=<11.3138 0.1529 0>, rot=<0 0 -0.2003 -0.9797>}
 neck_02 {parent=neck_01, loc=<4.5627 0 0>, rot=<0 0 0.0504 -0.9987>}
 head {parent=neck_02, loc=<5.3914 0 0>, rot=<0 0 0.115 -0.9934>}
 dyn_hair_side_l_base {parent=head, loc=<2.257 4.025 -7.508>, rot=<0.1607 -0.9827 0.0631 -0.0673>}
 dyn_hair_side_01_l {parent=dyn_hair_side_l_base, loc=<0.4808 0.0217 -0.0023>, rot=<0 0 0 1>}
 dyn_hair_side_02_l {parent=dyn_hair_side_01_l, loc=<10.5582 0 0>, rot=<0.0099 0.0068 0.3865 -0.9222>}
 dyn_hair_side_03_l {parent=dyn_hair_side_02_l, loc=<4.7513 -0.1387 -0.0106>, rot=<-0.0008 -0.0165 0.0577 -0.9982>}
 dyn_hair_side_r_base {parent=head, loc=<2.257 4.0245 7.5077>, rot=<0.043 0.0816 0.9813 -0.1688>}
 dyn_hair_side_01_r {parent=dyn_hair_side_r_base, loc=<0.4808 -0.0217 0.0023>, rot=<0 0 0 1>}
 dyn_hair_side_02_r {parent=dyn_hair_side_01_r, loc=<10.5582 0 0>, rot=<-0.0099 0.0068 -0.3865 -0.9222>}
 dyn_hair_side_03_r {parent=dyn_hair_side_02_r, loc=<4.7513 -0.1387 -0.0106>, rot=<0.0008 -0.0165 -0.0577 -0.9982>}
 dyn_hair_side_bk_l_base {parent=head, loc=<7.1302 -5.0187 -4.6295>, rot=<0.0957 -0.9954 0 0>}
 dyn_hair_side_bk_01_l {parent=dyn_hair_side_bk_l_base, loc=<0.3444 0.0639 0>, rot=<0 0 0 1>}
 dyn_hair_side_bk_02_l {parent=dyn_hair_side_bk_01_l, loc=<7.7274 -0.3037 0.6968>, rot=<0.0588 0.0147 0.0601 -0.9964>}
 dyn_hair_side_bk_03_l {parent=dyn_hair_side_bk_02_l, loc=<8.1625 0.2252 0.0629>, rot=<-0.0202 -0.0129 0.0049 -0.9997>}
 dyn_hair_side_bk_04_l {parent=dyn_hair_side_bk_03_l, loc=<6.8219 0.0225 0.0281>, rot=<0.0057 -0.0208 -0.085 -0.9961>}
 dyn_hair_side_bk_05_l {parent=dyn_hair_side_bk_04_l, loc=<6.4095 0.0823 0.0318>, rot=<0.013 -0.0006 -0.0443 -0.9989>}
 dyn_hair_center_bk_base {parent=head, loc=<7.2097 -6.3219 0.3833>, rot=<0.144 -0.9894 0.0087 0.0139>}
 dyn_hair_center_bk_01 {parent=dyn_hair_center_bk_base, loc=<0.4251 0.1227 0.0111>, rot=<0 0 0 1>}
 dyn_hair_center_bk_02 {parent=dyn_hair_center_bk_01, loc=<7.9201 0.5099 0.2304>, rot=<0.0067 0.0149 -0.0053 -0.9999>}
 dyn_hair_center_bk_03 {parent=dyn_hair_center_bk_02, loc=<8.2378 0 0>, rot=<0 0 0.0023 -1>}
 dyn_hair_center_bk_04 {parent=dyn_hair_center_bk_03, loc=<8.2799 0 0>, rot=<0 0 -0.1277 -0.9918>}
 dyn_hair_center_bk_05 {parent=dyn_hair_center_bk_04, loc=<9.6093 0 0>, rot=<0 0 0 1>}
 dyn_hair_side_bk_r_base {parent=head, loc=<7.1305 -5.0187 4.6295>, rot=<0.0957 -0.9954 0 0>}
 dyn_hair_side_bk_01_r {parent=dyn_hair_side_bk_r_base, loc=<0.3451 0.064 0>, rot=<0 0 0 1>}
 dyn_hair_side_bk_02_r {parent=dyn_hair_side_bk_01_r, loc=<7.727 -0.3038 -0.6969>, rot=<-0.0588 -0.0147 0.0601 -0.9964>}
 dyn_hair_side_bk_03_r {parent=dyn_hair_side_bk_02_r, loc=<8.1626 0.2253 -0.0629>, rot=<0.0202 0.0129 0.0049 -0.9997>}
 dyn_hair_side_bk_04_r {parent=dyn_hair_side_bk_03_r, loc=<6.8218 0.0224 -0.0281>, rot=<-0.0057 0.0208 -0.085 -0.9961>}
 dyn_hair_side_bk_05_r {parent=dyn_hair_side_bk_04_r, loc=<6.4092 0.0823 -0.0318>, rot=<-0.013 0.0006 -0.0443 -0.9989>}
 extra_head {parent=head, loc=<0 -0.5415 0>, rot=<0 0 0 1>}
 clavicle_l {parent=spine_05, loc=<5.4831 -0.1053 -3.2946>, rot=<0.0465 -0.7289 -0.009 -0.6829>}
 upperarm_l {parent=clavicle_l, loc=<9.8658 -1.2136 -0.4105>, rot=<0.0164 -0.3245 -0.0269 -0.9454>}
 dyn_extra_upperarm_l {parent=upperarm_l, loc=<0.5 0 0>, rot=<0 0 0 1>}
 clavicle_r {parent=spine_05, loc=<5.4831 -0.1053 3.2946>, rot=<0.7289 0.0465 -0.6829 0.009>}
 upperarm_r {parent=clavicle_r, loc=<-9.8658 1.2136 0.4105>, rot=<0.0164 -0.3245 -0.0269 -0.9454>}
 dyn_extra_upperarm_r {parent=upperarm_r, loc=<-0.5 0 0>, rot=<0 0 0 1>}
 dyn_backpack_collision {parent=spine_05, loc=<10.9743 -22.8145 0>, rot=<0.0616 -0.7044 -0.0616 -0.7044>}
 SPR_HeadUp {type=Curve, loc=<0 0 0>, rot=<0 0 0 1>}
]
variables [Crouch {type=bool, default=False}]
