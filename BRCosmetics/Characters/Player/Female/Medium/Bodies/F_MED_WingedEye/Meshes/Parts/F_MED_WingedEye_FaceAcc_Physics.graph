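# Physics asset for the female WingedEye character's FaceAcc mesh, setting up hair dynamics and body collision.
# Five capsule chains (center back, back sides, front sides) use mostly tight swing/twist limits; back bases lock angularly to the head, side bases rotate freely.
# Cross-constraints tie neighboring chains together so the hair moves as a group.
# Capsules on head, neck, spine, shoulders, upper arms and an oversized backpack volume provide collision shapes.
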PhysicsAsset /BRCosmetics/Characters/Player/Female/Medium/Bodies/F_MED_WingedEye/Meshes/Parts/F_MED_WingedEye_FaceAcc_Physics
bodies {
 clavicle_l [
  {shapeType=Capsule, length=10.6, radius=8, shapeName=clavicle_l_capsule, center=<6.2526 0.1462 -2.2573>, rotation=<rot 0 -90 90>}
  {
   shapeType=Capsule
   length=14
   radius=6.8
   shapeName=clavicle_l_capsule
   center=<3.0864 18.2035 10.8042>
   rotation=<rot -24.5939 94.6292 11.0095>
  }
 ]
 clavicle_r [
  {shapeType=Capsule, length=10.6, radius=8, shapeName=clavicle_l_capsule, center=<-3.1925 -1.5071 2.2573>, rotation=<rot 0 -90 90>}
  {
   shapeType=Capsule
   length=14
   radius=6.8
   shapeName=clavicle_r_capsule
   center=<-4.2567 -18.2035 -10.8042>
   rotation=<rot 24.5939 -94.6292 -11.0095>
  }
 ]
 dyn_backpack_collision [
  {shapeType=Capsule, length=50, radius=30, shapeName=dyn_backpack_collision_capsule, center=<0 -51.045 0>, rotation=<rot 0 -90 90>}
  {shapeType=Capsule, length=50, radius=30, shapeName=dyn_backpack_collision_capsule, center=<0 -51.045 -30>, rotation=<rot 0 -90 90>}
 ]
 dyn_extra_upperarm_l [{shapeType=Capsule, length=13.5146, radius=6, shapeName=dyn_extra_upperarm_l_capsule, center=<10.5459 -0.3627 0>, rotation=<rot 90 0 0>}]
 dyn_extra_upperarm_r [{shapeType=Capsule, length=13.5146, radius=6, shapeName=dyn_extra_upperarm_r_capsule, center=<-10.5459 0.3627 0>, rotation=<rot 90 0 0>}]
 dyn_hair_center_bk_01 [{shapeType=Capsule, length=6, radius=3, shapeName=dyn_hair_center_bk_01_capsule, center=<5 0 0>, rotation=<rot 90 0 0>}]
 dyn_hair_center_bk_02 [{shapeType=Capsule, length=3, radius=2, shapeName=dyn_hair_center_bk_02_capsule, center=<5 0 0>, rotation=<rot 90 0 0>}]
 dyn_hair_center_bk_03 [{shapeType=Capsule, length=3, radius=2, shapeName=dyn_hair_center_bk_03_capsule, center=<5 0 0>, rotation=<rot 90 0 0>}]
 dyn_hair_center_bk_04 [{shapeType=Capsule, length=3, radius=2, shapeName=dyn_hair_center_bk_04_capsule, center=<5 0 0>, rotation=<rot 90 0 0>}]
 dyn_hair_center_bk_05 [{shapeType=Capsule, length=3, radius=3, shapeName=dyn_hair_center_bk_05_capsule, center=<5 0 0>, rotation=<rot 90 0 0>}]
 dyn_hair_center_bk_base [{shapeType=Capsule, length=1, radius=1, shapeName=dyn_hair_center_bk_base_capsule}]
 dyn_hair_side_01_l [{shapeType=Capsule, length=7, radius=3, shapeName=dyn_hair_side_01_r_capsule, center=<6 0 0>, rotation=<rot 90 180 0>}]
 dyn_hair_side_01_r [{shapeType=Capsule, length=8, radius=1, shapeName=dyn_hair_side_01_r_capsule, center=<6 0 0>, rotation=<rot 90 0 0>}]
 dyn_hair_side_02_l [{shapeType=Capsule, length=4, radius=2, shapeName=dyn_hair_side_02_r_capsule, center=<2 0 0>, rotation=<rot 90 -179.9998 0>}]
 dyn_hair_side_02_r [{shapeType=Capsule, length=4, radius=2, shapeName=dyn_hair_side_02_r_capsule, center=<2 0 0>, rotation=<rot 90 0 0>}]
 dyn_hair_side_03_l [{shapeType=Capsule, length=6, radius=2, shapeName=dyn_hair_side_03_r_capsule, center=<2 0 0>, rotation=<rot 90 180 0>}]
 dyn_hair_side_03_r [{shapeType=Capsule, length=5.5, radius=2.5, shapeName=dyn_hair_side_03_r_capsule, center=<2 -0.6499 0>, rotation=<rot 90 -0.0003 0>}]
 dyn_hair_side_bk_01_l [{shapeType=Capsule, length=6, radius=3, shapeName=dyn_hair_side_bk_01_l_capsule, center=<5 0 0>, rotation=<rot 90 0 0>}]
 dyn_hair_side_bk_01_r [{shapeType=Capsule, length=6, radius=3, shapeName=dyn_hair_side_bk_01_r_capsule, center=<5 0 0>, rotation=<rot 90 0.0001 0>}]
 dyn_hair_side_bk_02_l [{shapeType=Capsule, length=3, radius=3, shapeName=dyn_hair_side_bk_02_l_capsule, center=<5 1.4458 0.6786>, rotation=<rot 90 0 0>}]
 dyn_hair_side_bk_02_r [{shapeType=Capsule, length=3, radius=3, shapeName=dyn_hair_side_bk_02_r_capsule, center=<5 1.4458 -0.6786>, rotation=<rot 90 0 0>}]
 dyn_hair_side_bk_03_l [{shapeType=Capsule, length=3, radius=3, shapeName=dyn_hair_side_bk_03_l_capsule, center=<5 0 1.1595>, rotation=<rot 90 0 0>}]
 dyn_hair_side_bk_03_r [{shapeType=Capsule, length=3, radius=3, shapeName=dyn_hair_side_bk_03_r_capsule, center=<5 0 -1.1595>, rotation=<rot 90 0 0>}]
 dyn_hair_side_bk_04_l [{shapeType=Capsule, length=3, radius=3, shapeName=dyn_hair_side_bk_04_l_capsule, center=<5 0 1.1595>, rotation=<rot 90 0 0>}]
 dyn_hair_side_bk_04_r [{shapeType=Capsule, length=3, radius=3, shapeName=dyn_hair_side_bk_04_r_capsule, center=<5 0 -1.1595>, rotation=<rot 90 0 0>}]
 dyn_hair_side_bk_05_l [{shapeType=Capsule, length=3, radius=3, shapeName=dyn_hair_side_bk_05_l_capsule, center=<5 0 0>, rotation=<rot 90 0 0>}]
 dyn_hair_side_bk_05_r [{shapeType=Capsule, length=3, radius=3, shapeName=dyn_hair_side_bk_05_r_capsule, center=<5 0 -1.1595>, rotation=<rot 90 0 0>}]
 dyn_hair_side_bk_l_base [{shapeType=Capsule, length=1, radius=1, shapeName=dyn_hair_side_bk_l_base_capsule}]
 dyn_hair_side_bk_r_base [{shapeType=Capsule, length=1, radius=1, shapeName=dyn_hair_side_bk_r_base_capsule}]
 dyn_hair_side_l_base [{shapeType=Capsule, length=2, radius=2, shapeName=dyn_hair_side_r_base_capsule, rotation=<rot 90 -179.9993 0>}]
 dyn_hair_side_r_base [{shapeType=Capsule, length=2, radius=2, shapeName=dyn_hair_side_r_base_capsule, rotation=<rot 90 -0.0007 0>}]
 head [{shapeType=Capsule, length=6.5, radius=5.4026, shapeName=head_capsule, center=<0.3063 1.1903 0>, rotation=<rot 90 179.9209 179.9219>}]
 neck_01 [{shapeType=Capsule, length=0, radius=5, shapeName=neck_01_capsule, center=<1.2134 0.1918 0>, rotation=<rot 90 0 0>}]
 neck_02 [
  {shapeType=Capsule, length=50, radius=15, shapeName=spine_05_capsule, center=<-8.4615 12.9244 0>}
  {shapeType=Capsule, length=76.3, radius=27, shapeName=neck_02_capsule, center=<5.4574 -5.8367 -40.8803>, rotation=<rot 0 -65.0008 90>}
  {shapeType=Capsule, length=77.8, radius=28, shapeName=neck_02_capsule, center=<1.9272 -8.9546 41.4654>, rotation=<rot 0 -65.0008 90>}
 ]
 spine_01 [{shapeType=Capsule, length=20, radius=12, shapeName=spine_01_capsule, rotation=<rot 0 -90 0>}]
 spine_02 [{shapeType=Capsule, length=13.7063, radius=12.2937, shapeName=spine_02_capsule, center=<6.0567 1.2153 0>}]
 spine_03 [{shapeType=Capsule, length=15.9887, radius=10.0113, shapeName=spine_02_capsule, center=<6.0567 1.1414 0>}]
 spine_04 [
  {shapeType=Capsule, length=17.8, radius=9.1, shapeName=spine_04_capsule, center=<4.0135 1.6435 0>, rotation=<rot 0 -90 0>}
  {shapeType=Capsule, length=20.8, radius=6.1, shapeName=spine_04_capsule, center=<14.0135 0.2077 0>, rotation=<rot 0 -90 0>}
 ]
 spine_05 [
  {shapeType=Capsule, length=16.7, radius=8.5, shapeName=spine_05_capsule, center=<2.8857 0.1819 0>, rotation=<rot 0 -90 0>}
  {shapeType=Capsule, length=3, radius=10, shapeName=spine_05_capsule, center=<7.3391 1.1263 0>, rotation=<rot 0 -90 0>}
  {shapeType=Capsule, length=10, radius=8, shapeName=spine_05_capsule, center=<5.4121 -0.7836 0>, rotation=<rot 0 -90 0>}
 ]
 upperarm_l [{shapeType=Capsule, length=5, radius=1, shapeName=upperarm_l_capsule, center=<11.2314 -0.3627 0>, rotation=<rot 90 0 0>}]
 upperarm_r [{shapeType=Capsule, length=13.5146, radius=1, shapeName=upperarm_l_capsule, center=<-11.2314 0.3627 0>, rotation=<rot 90 0 0>}]
}
constraints [
 {bone1Name=dyn_hair_center_bk_01, bone2Name=dyn_hair_center_bk_base, swing1LimitDegrees=10, swing2LimitDegrees=10, twistLimitDegrees=1}
 {bone1Name=dyn_hair_center_bk_02, bone2Name=dyn_hair_center_bk_01, swing1LimitDegrees=10, swing2LimitDegrees=10, twistLimitDegrees=1}
 {bone1Name=dyn_hair_center_bk_03, bone2Name=dyn_hair_center_bk_02, swing1LimitDegrees=30, swing2LimitDegrees=10, twistLimitDegrees=1}
 {bone1Name=dyn_hair_center_bk_04, bone2Name=dyn_hair_center_bk_03, swing1LimitDegrees=10, swing2LimitDegrees=10, twistLimitDegrees=1}
 {bone1Name=dyn_hair_center_bk_05, bone2Name=dyn_hair_center_bk_04, swing1LimitDegrees=10, swing2LimitDegrees=10, twistLimitDegrees=1}
 {bone1Name=dyn_hair_side_bk_01_l, bone2Name=dyn_hair_side_bk_l_base, swing1LimitDegrees=45, swing2LimitDegrees=10, twistLimitDegrees=1}
 {bone1Name=dyn_hair_side_bk_02_l, bone2Name=dyn_hair_side_bk_01_l, swing1LimitDegrees=10, swing2LimitDegrees=10, twistLimitDegrees=1}
 {bone1Name=dyn_hair_side_bk_03_l, bone2Name=dyn_hair_side_bk_02_l, swing1LimitDegrees=10, swing2LimitDegrees=5, twistLimitDegrees=1}
 {bone1Name=dyn_hair_side_bk_04_l, bone2Name=dyn_hair_side_bk_03_l, swing1LimitDegrees=10, swing2LimitDegrees=5, twistLimitDegrees=1}
 {bone1Name=dyn_hair_side_bk_05_l, bone2Name=dyn_hair_side_bk_04_l, swing1LimitDegrees=30, swing2LimitDegrees=10, twistLimitDegrees=1}
 {bone1Name=dyn_hair_side_bk_01_r, bone2Name=dyn_hair_side_bk_r_base, swing1LimitDegrees=10, swing2LimitDegrees=10, twistLimitDegrees=1}
 {bone1Name=dyn_hair_side_bk_02_r, bone2Name=dyn_hair_side_bk_01_r, swing1LimitDegrees=10, swing2LimitDegrees=10, twistLimitDegrees=1}
 {bone1Name=dyn_hair_side_bk_03_r, bone2Name=dyn_hair_side_bk_02_r, swing1LimitDegrees=10, swing2LimitDegrees=10, twistLimitDegrees=1}
 {bone1Name=dyn_hair_side_bk_04_r, bone2Name=dyn_hair_side_bk_03_r, swing1LimitDegrees=30, swing2LimitDegrees=10, twistLimitDegrees=1}
 {bone1Name=dyn_hair_side_bk_05_r, bone2Name=dyn_hair_side_bk_04_r, swing1LimitDegrees=30, swing2LimitDegrees=10, twistLimitDegrees=1}
 {bone1Name=dyn_hair_side_01_l, bone2Name=dyn_hair_side_l_base, swing1LimitDegrees=45, swing2LimitDegrees=15, twistLimitDegrees=5}
 {bone1Name=dyn_hair_side_02_l, bone2Name=dyn_hair_side_01_l, swing1LimitDegrees=15, swing2LimitDegrees=15, twistLimitDegrees=5}
 {bone1Name=dyn_hair_side_03_l, bone2Name=dyn_hair_side_02_l, swing1LimitDegrees=15, swing2LimitDegrees=15, twistLimitDegrees=5}
 {bone1Name=dyn_hair_side_01_r, bone2Name=dyn_hair_side_r_base, swing1LimitDegrees=15, swing2LimitDegrees=15, twistLimitDegrees=5}
 {bone1Name=dyn_hair_side_02_r, bone2Name=dyn_hair_side_01_r, swing1LimitDegrees=15, swing2LimitDegrees=15, twistLimitDegrees=5}
 {bone1Name=dyn_hair_side_03_r, bone2Name=dyn_hair_side_02_r, swing1LimitDegrees=15, swing2LimitDegrees=15, twistLimitDegrees=5}
 {
  bone1Name=dyn_hair_center_bk_base
  bone2Name=head
  swing1LimitDegrees=10
  swing1Motion=Locked
  swing2LimitDegrees=10
  swing2Motion=Locked
  twistLimitDegrees=1
  twistMotion=Locked
 }
 {
  bone1Name=dyn_hair_side_bk_l_base
  bone2Name=head
  swing1LimitDegrees=10
  swing1Motion=Locked
  swing2LimitDegrees=10
  swing2Motion=Locked
  twistLimitDegrees=1
  twistMotion=Locked
 }
 {
  bone1Name=dyn_hair_side_bk_r_base
  bone2Name=head
  swing1LimitDegrees=10
  swing1Motion=Locked
  swing2LimitDegrees=10
  swing2Motion=Locked
  twistLimitDegrees=1
  twistMotion=Locked
 }
 {
  bone1Name=dyn_hair_side_l_base
  bone2Name=head
  swing1LimitDegrees=10
  swing1Motion=Free
  swing2LimitDegrees=10
  swing2Motion=Free
  twistLimitDegrees=1
  twistMotion=Free
 }
 {
  bone1Name=dyn_hair_side_r_base
  bone2Name=head
  swing1LimitDegrees=10
  swing1Motion=Free
  swing2LimitDegrees=10
  swing2Motion=Free
  twistLimitDegrees=1
  twistMotion=Free
 }
 {
  bone1Name=dyn_hair_center_bk_04
  bone2Name=dyn_hair_side_bk_04_l
  swing1LimitDegrees=45
  swing1Motion=Free
  swing2LimitDegrees=45
  swing2Motion=Free
  twistLimitDegrees=45
  twistMotion=Free
 }
 {
  bone1Name=dyn_hair_side_bk_04_r
  bone2Name=dyn_hair_center_bk_04
  swing1LimitDegrees=45
  swing1Motion=Free
  swing2LimitDegrees=45
  swing2Motion=Free
  twistLimitDegrees=45
  twistMotion=Free
 }
 {
  bone1Name=head
  bone2Name=spine_01
  swing1LimitDegrees=45
  swing1Motion=Locked
  swing2LimitDegrees=45
  swing2Motion=Locked
  twistLimitDegrees=45
  twistMotion=Locked
 }
 {
  bone1Name=clavicle_l
  bone2Name=spine_05
  swing1LimitDegrees=45
  swing1Motion=Locked
  swing2LimitDegrees=45
  swing2Motion=Locked
  twistLimitDegrees=45
  twistMotion=Locked
 }
 {
  bone1Name=upperarm_l
  bone2Name=clavicle_l
  swing1LimitDegrees=45
  swing1Motion=Locked
  swing2LimitDegrees=45
  swing2Motion=Locked
  twistLimitDegrees=45
  twistMotion=Locked
 }
 {
  bone1Name=dyn_extra_upperarm_l
  bone2Name=upperarm_l
  swing1LimitDegrees=45
  swing1Motion=Locked
  swing2LimitDegrees=45
  swing2Motion=Locked
  twistLimitDegrees=45
  twistMotion=Locked
 }
 {
  bone1Name=clavicle_r
  bone2Name=spine_05
  swing1LimitDegrees=45
  swing1Motion=Locked
  swing2LimitDegrees=45
  swing2Motion=Locked
  twistLimitDegrees=45
  twistMotion=Locked
 }
 {
  bone1Name=upperarm_r
  bone2Name=clavicle_r
  swing1LimitDegrees=45
  swing1Motion=Locked
  swing2LimitDegrees=45
  swing2Motion=Locked
  twistLimitDegrees=45
  twistMotion=Locked
 }
 {
  bone1Name=dyn_extra_upperarm_r
  bone2Name=upperarm_r
  swing1LimitDegrees=45
  swing1Motion=Locked
  swing2LimitDegrees=45
  swing2Motion=Locked
  twistLimitDegrees=45
  twistMotion=Locked
 }
 {
  bone1Name=dyn_backpack_collision
  bone2Name=spine_05
  swing1LimitDegrees=45
  swing1Motion=Locked
  swing2LimitDegrees=45
  swing2Motion=Locked
  twistLimitDegrees=45
  twistMotion=Locked
 }
 {bone1Name=dyn_hair_side_02_r, bone2Name=dyn_hair_side_bk_02_r, swing1LimitDegrees=45, swing2LimitDegrees=45, twistLimitDegrees=1}
 {bone1Name=dyn_hair_side_02_l, bone2Name=dyn_hair_side_bk_02_l, swing1LimitDegrees=45, swing2LimitDegrees=45, twistLimitDegrees=1}
 {bone1Name=dyn_hair_side_03_r, bone2Name=dyn_hair_side_bk_02_r, swing1LimitDegrees=10, swing2LimitDegrees=10, twistLimitDegrees=10}
 {bone1Name=dyn_hair_side_03_l, bone2Name=dyn_hair_side_bk_02_l, swing1LimitDegrees=10, swing2LimitDegrees=10, twistLimitDegrees=10}
]
